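# Animation sharing state blueprint for the TidalCrane Swarmer creature skeleton, producing the pose for one animation state in a swarm.
# A single looping sequence player feeds the output pose, taking its clip from AnimationToPlay and its start position from PermutationTimeOffset, so permutations can play out of phase.
# Both values are inherited rather than declared here, and the update event has no logic.
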AnimBlueprint /ProtoSwarm/Assets/AnimSharing/SwarmState
target_skeleton=/ProtoSwarm/Assets/TidalCrane_Swarmer/Meshes/TidalCrane_Swarmer_Creature_Skeleton
graphs [
 AnimGraph {nodes [Output Pose(Result=Sequence Player!(
     Sequence:Anim Sequence Base Object Reference=Get AnimationToPlay() -> (AnimationToPlay:Anim Sequence Object Reference)
     PlayRate:float=1
     StartPosition:float=Get PermutationTimeOffset() -> (PermutationTimeOffset:float)
     bLoopAnimation:bool=True
    ) -> (Pose)) {layerGroup=None}]}
 EventGraph {nodes [Event BlueprintUpdateAnimation()]}
]
variables []
