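# Sound cue for the crystal death burst of a Proto Bug in ProtoSwarm, classed as a creature sound.
# A Random node picks one of four equally weighted burst waves and is excluded from branch culling.
# A root Modulator varies pitch and volume on each play, and the cue's volume is boosted.
# A shared death-burst attenuation asset shapes falloff, while global and local concurrency sets govern how many bursts overlap.
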SoundCue /ProtoSwarm/Assets/Audio/SC_Proto_Bug_Crystal_Death_Burst
root=Modulator_0
nodes [
 {id=Modulator_0, children [Random_1], properties {pitchMax=1.2, volumeMax=1}}
 {
  id=Random_1
  children [WavePlayer_4,WavePlayer_5,WavePlayer_6,WavePlayer_7]
  properties {bShouldExcludeFromBranchCulling=true, weights [1,1,1,1]}
 }
 {id=WavePlayer_4, properties {soundWaveAssetPtr=/ProtoSwarm/Assets/Audio/SoundWaves/Proto_Bug_Crystal_Death_Burst_01}}
 {id=WavePlayer_5, properties {soundWaveAssetPtr=/ProtoSwarm/Assets/Audio/SoundWaves/Proto_Bug_Crystal_Death_Burst_02}}
 {id=WavePlayer_6, properties {soundWaveAssetPtr=/ProtoSwarm/Assets/Audio/SoundWaves/Proto_Bug_Crystal_Death_Burst_03}}
 {id=WavePlayer_7, properties {soundWaveAssetPtr=/ProtoSwarm/Assets/Audio/SoundWaves/Proto_Bug_Crystal_Death_Burst_04}}
]
properties {
 attenuationSettings=/ProtoSwarm/Assets/Audio/Settings/ATT_Proto_Bug_Death_Burst
 maxDistance=16750
 soundClassObject=/Game/Sounds/FortSoundClassesAndModes/Fort_Creatures
 volumeMultiplier=1.577
 concurrencyOverrides {maxCount=4}
 concurrencySet [/ProtoSwarm/Assets/Audio/Settings/CONC_Proto_Bug_Death_Burst_Global, /ProtoSwarm/Assets/Audio/Settings/CONC_Proto_Bug_Death_Burst_Local]
}
