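# Sound cue for the ground impact of a crystal bug's death projectile in ProtoSwarm.
# A Random node picks one of four equally weighted impact waves and feeds a Modulator that varies pitch on each play.
# It plays in the creatures sound class, with death-burst attenuation and a concurrency set for ground impacts to limit overlapping plays.
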SoundCue /ProtoSwarm/Assets/Audio/SC_Proto_Bug_Crystal_Death_Prj_GroundImpact
root=Modulator_0
nodes [
 {id=Modulator_0, children [Random_0], properties {pitchMax=1.2, volumeMax=1}}
 {
  id=Random_0
  children [WavePlayer_0,WavePlayer_1,WavePlayer_2,WavePlayer_3]
  properties {bShouldExcludeFromBranchCulling=true, weights [1,1,1,1]}
 }
 {id=WavePlayer_0, properties {soundWaveAssetPtr=/ProtoSwarm/Assets/Audio/SoundWaves/Proto_Bug_Crystal_Death_GroundImpact_04}}
 {id=WavePlayer_1, properties {soundWaveAssetPtr=/ProtoSwarm/Assets/Audio/SoundWaves/Proto_Bug_Crystal_Death_GroundImpact_03}}
 {id=WavePlayer_2, properties {soundWaveAssetPtr=/ProtoSwarm/Assets/Audio/SoundWaves/Proto_Bug_Crystal_Death_GroundImpact_02}}
 {id=WavePlayer_3, properties {soundWaveAssetPtr=/ProtoSwarm/Assets/Audio/SoundWaves/Proto_Bug_Crystal_Death_GroundImpact_01}}
]
properties {
 attenuationSettings=/ProtoSwarm/Assets/Audio/Settings/ATT_Proto_Bug_Death_Burst
 maxDistance=16750
 soundClassObject=/Game/Sounds/FortSoundClassesAndModes/Fort_Creatures
 volumeMultiplier=1.577
 concurrencyOverrides {maxCount=4}
 concurrencySet [/ProtoSwarm/Assets/Audio/Settings/CONC_Proto_Bug_Death_Burst_Prj_ImpactGround]
}
