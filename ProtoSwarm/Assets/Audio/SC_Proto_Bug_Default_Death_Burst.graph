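# Sound cue named for the default death burst of a ProtoSwarm bug, routed to the Fort_Creatures sound class.
# The cue shows no root or sound nodes, so only its playback settings are defined here.
# It uses a dedicated death-burst attenuation asset and boosts volume.
# Separate global and local concurrency assets are assigned to limit simultaneous bursts, alongside a concurrency override block with a small voice cap.
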SoundCue /ProtoSwarm/Assets/Audio/SC_Proto_Bug_Default_Death_Burst
root=null
nodes []
properties {
 attenuationSettings=/ProtoSwarm/Assets/Audio/Settings/ATT_Proto_Bug_Death_Burst
 maxDistance=16750
 soundClassObject=/Game/Sounds/FortSoundClassesAndModes/Fort_Creatures
 volumeMultiplier=1.577
 concurrencyOverrides {maxCount=4}
 concurrencySet [/ProtoSwarm/Assets/Audio/Settings/CONC_Proto_Bug_Death_Burst_Global, /ProtoSwarm/Assets/Audio/Settings/CONC_Proto_Bug_Death_Burst_Local]
}
