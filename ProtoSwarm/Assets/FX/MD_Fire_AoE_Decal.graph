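# Translucent unlit deferred decal that draws animated, glowing fire for an area effect.
# Scrolling volume noise, sampled in actor-relative coordinates warped by a sine of distance, blends with world-space Voronoi noise for emissive detail.
# Decal Color RGB carries a Niagara-driven range offset instead of color.
# Lifetime opacity and an exponential radial gradient shape the mask, while switches boost alpha on mobile-class platforms and add detail on higher quality tiers.
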Material /ProtoSwarm/Assets/FX/MD_Fire_AoE_Decal
graph {
 properties {
  bAutomaticallySetUsageInEditor=false
  blendMode=BLEND_Translucent
  materialDomain=MD_DeferredDecal
  pixelDepthOffsetMode=PDOM_Legacy
  shadingModel=MSM_Unlit
  usedShadingModels=Unlit
 }
 nodes [
  n1 = VectorParameter() {group=Emissive, parameterName=Center Noise Color, defaultValue=<rgba 30 4.4229 0 1>}
  n2 = Multiply(A=NamedRerouteUsage(), B=ScalarParameter() {defaultValue=10, group=Emissive, parameterName=Additional Emissive Detail Intensity}) {constB=2}
  n3 = OneMinus(None=DecalLifetimeOpacity().Opacity)
  n4 = Multiply(A=Subtract(A=Multiply(A=@n3, B=@n3)) {constB=0.77}, B=NamedRerouteUsage())
  n5 = Multiply(A=@n4, B=ScalarParameter() {defaultValue=4, parameterName=Alpha Adjustment for Mobile})
  n6 = Multiply(
   A=MaterialFunctionCall("Xb1,PS4"=@n4, "PC-High&&Epic,XSX,PS5"=@n4, "PC-Perf,Android"=@n5, Switch=@n5, "PC-Low(NotPerf)&&Medium"=@n4) {materialFunction=&PlatformSwitch_1D}.Result
   B=@n1.RGB
  )
  n7 = Add(A=@n6, B=Multiply(A=@n1.RGB, B=Add(A=@n2)))
  MaterialFunctionCall(Emissive=MaterialFunctionCall!(
    Low _ PC Low and Med=@n6
    "High _ XB1, PS4"=@n7
    "Epic _ PC High, XSX, PS5"=@n7
    --------=@n6
    "Mobile Low _ PC Perf, Android"=@n6
    Mobile Med _ Switch=@n6
   ) {materialFunction=/Game/Global/MaterialFunctions/TechArt/MF_Fortnite_QualitySwitch_Vector3}.Result) {materialFunction=&MF_FX_EmissiveAdjustments}
  Multiply(
   A=Multiply(A=DecalLifetimeOpacity().Opacity, B=NamedRerouteUsage())
   B=MaterialFunctionCall() {materialFunction=&RadialGradientExponential}.RadialGradientExponential
  )
  n8 = NamedRerouteUsage()
  NamedRerouteDeclaration(None=Multiply(A=ScalarParameter() {defaultValue=100, group=Sine, parameterName=Gradient Center Intensity}, B=Saturate(None=@n8))) {name=Mask Main_1, nodeColor=<rgba 1 0 0.9375 1>}
  n9 = WorldPosition()
  n10 = Distance(A=Subtract(A=ActorPositionWS(), B=@n9.XYZ), B=Constant3Vector())
  n11 = ScalarParameter() {defaultValue=0.0004, parameterName=Distance Adjustment}
  n12 = Add(A=Sine(None=Multiply(A=@n10, B=@n11) {constB=0.01}), B=ScalarParameter() {defaultValue=0.88, parameterName=Param_3})
  n13 = Multiply(A=DecalColor() {desc="RGB= Result of Decal Range Adjustment via Niagara\r\n!! It's NOT Color value in this Material.\r\n\r\nA= Second Opacity"}.RGB) {constB=-1}
  n14 = Multiply(
   A=Power(Base=Multiply(A=@n12) {constB=0.5}, Exp=ScalarParameter() {defaultValue=-4.1}) {constExponent=-4}
   B=Subtract(A=ActorPositionWS(), B=Add(A=@n9.XYZ, B=@n13))
  )
  n15 = Multiply(A=Time(), B=Constant3Vector() {constant=<rgba 0 0 -25 0>})
  n16 = TextureSample(UVs=Multiply(A=Add(A=@n14, B=@n15), B=ScalarParameter() {defaultValue=0.0016, group=Volume Noise, parameterName=UVW_Scale})) {samplerType=SAMPLERTYPE_LinearColor, texture=/Game/Global/MaterialFunctions/VT_ClothSimWPO}
  NamedRerouteDeclaration(None=Multiply(A=@n16.R, B=@n16.R)) {name=Noise, nodeColor=<rgba 1 0.2344 0 1>}
  n17 = ScalarParameter() {defaultValue=0.0065, group=Volume Noise, parameterName=UVW_Scale_High}
  n18 = WorldPosition()
  n19 = Time()
  n20 = Constant3Vector() {constant=<rgba 0.2 0.13 0.7 0>}
  n21 = TextureSample(UVs=Add(A=Multiply(A=@n17, B=@n18.XYZ), B=Multiply(A=@n19, B=@n20))) {samplerType=SAMPLERTYPE_LinearColor, texture=/Game/Effects/Explosion_Update/Textures/VT_Voronoi_Seamless}
  n22 = LinearInterpolate(A=NamedRerouteUsage(), B=Multiply(A=@n21.R, B=@n21.R), Alpha=ScalarParameter() {defaultValue=0.5, parameterName=Lerp})
  n23 = Multiply(A=@n22, B=@n22)
  NamedRerouteDeclaration(None=Multiply(A=@n23, B=@n23)) {name=Emissive Detail, nodeColor=<rgba 0 1 0.7422 1>}
 ]
}
