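# Masked, two-sided lit material for a goop aura effect on mesh particles, built on the MF_TwistedGoo function from the Mastermind character.
# Adds vector-noise World Position Offset in switchable local or world space, tuned by Dynamic Parameters, masked by vertex red, with a constant replacing noise on ES2/ES3.1.
# Particle alpha drives Opacity Mask, while emissive gains a blue tint and is saturated before the shared FX emissive-adjustment function.
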Material /ProtoSwarm/Assets/FX/M_Aura_Goop_VFX
graph {
 properties {
  bUseMaterialAttributes=true
  bUsedWithMeshParticles=true
  bUsedWithNiagaraMeshParticles=true
  blendMode=BLEND_Masked
  pixelDepthOffsetMode=PDOM_Legacy
  twoSided=true
  usedShadingModels=Default Lit
 }
 nodes [
  n1 = GetMaterialAttributes(None=MaterialFunctionCall!(In, Mask, isVehicle, isMech) {materialFunction=/Game/Characters/Player/Male/Medium/Bodies/M_MED_Mastermind/Materials/MF_TwistedGoo}.Result) {attributeGetTypes [#1]}
  n2 = DynamicParameter() {defaultValue=<rgba 0.2 25 1.5 1>, paramNames ["Curl Speed (2)", "Curl Scale (10)", "Curl Intensity (3)", Param4]}
  n3 = Subtract(
   A=StaticSwitchParameter(
    True=MaterialFunctionCall() {materialFunction=&LocalPosition}.Local Position
    False=WorldPosition().XYZ
   ) {parameterName=UseLocalPosition?}
   B=Multiply(A=@n2."Curl Speed (2)", B=Time())
  )
  n4 = Multiply(A=@n2."Curl Intensity (3)", B=VectorNoise(World Position=Divide(A=@n3, B=@n2."Curl Scale (10)")) {noiseFunction=VNF_VectorALU})
  n5 = Constant() {r=1}
  n6 = StaticSwitchParameter(True=Multiply(A=VertexColor().R, B=FeatureLevelSwitch(Default=@n4, ES2=@n5, ES3_1=@n5, SM4_REMOVED=@n4)), False=Constant()) {defaultValue=true, parameterName=USE_WPO}
  n7 = GetMaterialAttributes(None=SetMaterialAttributes!(MaterialAttributes=@n1.MaterialAttributes, Opacity Mask=ParticleColor().A, World Position Offset=Add(A=@n1.World Position Offset, B=@n6)) {attributeSetTypes [#2, #1]}) {attributeGetTypes [#3]}
  n8 = Add(A=@n7.Emissive Color, B=Constant3Vector() {constant=<rgba 0 0.0273 0.5 1>}) {constB=0.25}
  MaterialFunctionCall!(In=SetMaterialAttributes!(MaterialAttributes=@n7.MaterialAttributes, Emissive Color=Saturate(None=@n8)) {attributeSetTypes [#3]}) {materialFunction=/Game/Effects/Niagara/MasterMaterials/Functions/MF_FX_EmissiveAdjustments_MA}
  ParticlePositionWS()
  ScalarParameter() {defaultValue=10, parameterName=Curl_Scale}
  ScalarParameter() {defaultValue=2, parameterName=Curl_Speed}
  ScalarParameter() {defaultValue=3, parameterName=Curl_Intensity}
 ]
}
