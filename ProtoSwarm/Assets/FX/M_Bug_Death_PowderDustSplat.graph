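# Unlit translucent Niagara sprite material for a powdery dust splat effect on bug death.
# Emissive multiplies particle color by a brightness lerp from the contrast-boosted grayscale splat texture, then passes through a shared emissive adjustment function.
# Opacity blends that texture with a repeating vertical UV gradient, using particle alpha as an erosion threshold.
# A dynamic parameter scales opacity, and depth fade softens intersections except on mobile feature levels.
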Material /ProtoSwarm/Assets/FX/M_Bug_Death_PowderDustSplat
graph {
 properties {
  bAutomaticallySetUsageInEditor=false
  bUsedWithNiagaraSprites=true
  blendMode=BLEND_Translucent
  shadingModel=MSM_Unlit
  usedShadingModels=Unlit
 }
 nodes [
  Constant() {r=0.5}
  n1 = ParticleColor()
  n2 = TextureSampleParameter2D() {parameterName=TEx1, samplerType=SAMPLERTYPE_LinearGrayscale, texture=&/T_BugDeathSplat}
  n3 = Power(Base=@n2.R) {constExponent=5}
  MaterialFunctionCall(Emissive=Multiply(A=@n1.RGB, B=LinearInterpolate(Alpha=@n3) {constA=0.9, constB=1.6})) {materialFunction=&MF_FX_EmissiveAdjustments}
  n4 = Power(Base=@n2.R) {constExponent=5}
  n5 = Saturate(None=Multiply(A=LinearInterpolate(A=@n2.R, B=@n4)))
  n6 = Multiply(A=TextureCoordinate(), B=ScalarParameter() {defaultValue=2, parameterName=TileNumber})
  n7 = Constant() {r=1}
  n8 = Multiply(A=MaterialFunctionCall(Float2=Modulo!(A=@n6, B=@n7)) {materialFunction=&BreakOutFloat2Components}.G)
  n9 = Power(Base=OneMinus(None=Power(Base=@n8) {constExponent=1}))
  n10 = Subtract(A=LinearInterpolate(A=Multiply(A=@n5, B=@n9), B=@n9) {constAlpha=0.7}, B=LinearInterpolate(Alpha=@n1.A) {constA=1, constB=-1})
  Max(A=@n10, B=LinearInterpolate(A=@n10, B=Divide(A=@n10, B=@n1.A)) {constAlpha=0.6})
  n11 = Multiply(A=Multiply(A=Power(Base=@n2.R) {constExponent=3}) {constB=2}, B=@n10)
  n12 = DepthFade(FadeDistance=ScalarParameter() {defaultValue=70, parameterName=FadeDistance})
  n13 = Constant() {r=1}
  Multiply(
   A=Multiply(A=Divide(A=@n11, B=@n1.A), B=DynamicParameter() {paramNames [Opacity,Param2,Param3,Param4]}.Opacity)
   B=FeatureLevelSwitch(Default=@n12, ES2=@n13, ES3_1=@n13, SM4_REMOVED=@n13, SM5=@n12, SM6=@n12)
  )
 ]
}
