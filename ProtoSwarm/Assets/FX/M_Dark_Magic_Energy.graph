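# Unlit translucent dark-magic energy material for Niagara mesh particles.
# Panning cell noise distorts a gradient texture's V coordinate, while noise-warped flame tiles sampled at three Twitch offsets add an RGB split.
# A hue-shifted curve-atlas gradient or solid color tints their sum, times particle color.
# Gradient, V-edge fade and particle alpha form opacity, depth-faded except on forward/mobile, and drive refraction; dynamic parameters control noise, refraction and optional offset.
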Material /ProtoSwarm/Assets/FX/M_Dark_Magic_Energy
graph {
 properties {
  bUseTranslucencyVertexFog=false
  bUsedWithNiagaraMeshParticles=true
  blendMode=BLEND_Translucent
  pixelDepthOffsetMode=PDOM_Legacy
  refractionMethod=RM_IndexOfRefraction
  shadingModel=MSM_Unlit
  usedShadingModels=Unlit
 }
 nodes [
  n1 = ParticleColor()
  n2 = ScalarParameter() {group=EmissiveNoise, parameterName=EmissiveNoise Rotation}
  n3 = MaterialFunctionCall(UVs=TextureCoordinate() {coordinateIndex=1}, "Rotation Angle (0-1)"=Divide(A=Floor(None=@n2)) {constB=360}) {materialFunction=&CustomRotator}
  n4 = VectorParameter() {group=EmissiveNoise, parameterName=EmissiveNoise UV, defaultValue=<rgba 1 1 0 0>}
  n5 = AppendVector(A=@n4.R, B=@n4.R)
  n6 = AppendVector(A=@n4.B, B=@n4.B)
  n7 = Time()
  n8 = TextureSampleParameter2D(UVs=Add(A=Multiply(A=@n3.Rotated Values, B=@n5), B=Multiply(A=@n6, B=@n7))) {
   group=EmissiveNoise
   parameterName=EmissiveNoise Tex
   samplerType=SAMPLERTYPE_Grayscale
   texture=/Game/Weapons/FORT_Melee/Pickaxe_Cube_Ninja_Male/FX/T_CellNoise
  }
  n9 = Multiply(
   A=ScalarParameter() {defaultValue=1, group=EmissiveNoise, parameterName=EmissiveNoise Intensity}
   B=DynamicParameter() {defaultValue=<rgba 0 1 1 1>, paramNames [UV Offset, Noise, Ref, Tile Intensity]}.Noise
  )
  n10 = Multiply(A=OneMinus(None=@n8.R), B=@n9)
  n11 = Subtract(A=@n10, B=Multiply(A=@n8.R, B=@n9))
  n12 = TextureCoordinate()
  n13 = MaterialFunctionCall(Float2=@n12) {materialFunction=&BreakOutFloat2Components}
  n14 = LinearInterpolate(A=@n12, B=AppendVector(A=@n13.R, B=OneMinus(None=@n13.G)), Alpha=ScalarParameter() {parameterName=Invert V})
  n15 = VectorParameter() {group=Emissive, parameterName=Emissive UV, defaultValue=<rgba 1 1 0 0>}
  n16 = AppendVector(A=@n15.R, B=@n15.R)
  n17 = StaticSwitchParameter(True=DynamicParameter() {defaultValue=<rgba 0 1 1 1>, paramNames [UV Offset, Noise, Ref, Tile Intensity]}.UV Offset, False=@n15.A) {group=Emissive, parameterName=Use UV Dynamic}
  n18 = Add(A=Multiply(A=Constant2Vector() {g=1}, B=Multiply(A=@n10, B=@n11)), B=Add(A=Multiply(A=@n14, B=@n16), B=AppendVector(A=@n15.B, B=@n17)))
  n19 = Multiply(
   A=MaterialFunctionCall!(TexCoord=@n18, ClampChannel=Constant2Vector() {g=1}) {materialFunction=/Game/Effects/Fort_Effects/MaterialFunctions/MF_FX_TilingMethod}.Result
   B=ChannelMaskParameter(None=TextureSampleParameter2D(UVs=@n18) {group=Emissive, parameterName=Emissive Tex, texture=/Game/Weapons/FORT_Melee/Pickaxe_Cube_Ninja_Male/FX/T_Offset_Gradient_01}.RGBA) {group=Emissive, maskChannel=Blue, parameterName=Emissive Channel, defaultValue=<rgba 0 0 1 0>}.R
  )
  n20 = ScalarParameter() {defaultValue=1, group=Emissive, parameterName=Emissive Mask}
  n21 = ComponentMask(None=TextureCoordinate()) {g=true}
  n22 = Multiply(A=@n21, B=@n21)
  n23 = OneMinus(None=@n21)
  n24 = Multiply(A=@n23, B=@n23)
  n25 = Saturate(None=Multiply(A=Multiply(A=@n22, B=@n24)) {constB=10})
  n26 = ScalarParameter() {defaultValue=5, parameterName=Mask Multiply}
  n27 = Multiply(A=@n1.A, B=Multiply(A=Multiply(A=@n19, B=@n20), B=Multiply(A=@n25, B=@n26) {constB=10}))
  n28 = ScalarParameter() {parameterName=Ref Intensity}
  n29 = DynamicParameter() {defaultValue=<rgba 0 1 1 1>, paramNames [UV Offset, Noise, Ref, Tile Intensity]}
  LinearInterpolate(Alpha=Multiply(A=@n27, B=Multiply(A=@n28, B=@n29.Ref))) {constA=1, constB=1.05}
  n30 = ComponentMask(None=TextureCoordinate()) {g=true}
  n31 = StaticSwitchParameter(
   True=VectorParameter() {parameterName=SolidColor, defaultValue=<rgba 0 0 0 1>}.RGB
   False=MaterialFunctionCall(
    Hue Shift Percentage=ScalarParameter() {parameterName=HueShift}
    Texture=CurveAtlasRowParameter(CurveTime=@n30) {atlas=&/CurveAtlas_DarkEnergy, curve=&/Curve_DarkEnergy_Color}
   ) {materialFunction=&HueShift}.Result
  ) {parameterName=UseSolidColor}
  n32 = TextureCoordinate()
  n33 = VectorParameter() {group=TileNoise, parameterName=TileNoise UV, defaultValue=<rgba 1 1 0 0>}
  n34 = AppendVector(A=@n33.R, B=@n33.R)
  n35 = AppendVector(A=@n33.B, B=@n33.B)
  n36 = Time()
  n37 = TextureSampleParameter2D(UVs=Add(A=Multiply(A=@n32, B=@n34), B=Multiply(A=@n35, B=@n36))) {
   group=TileNoise
   parameterName=TileNoise Tex
   samplerType=SAMPLERTYPE_Grayscale
   texture=&LowResBlurredNoise
  }
  n38 = TextureCoordinate()
  n39 = VectorParameter() {group=Tile, parameterName=Tile UV, defaultValue=<rgba 1 1 0 0>}
  n40 = AppendVector(A=@n39.R, B=@n39.R)
  n41 = AppendVector(A=@n39.B, B=@n39.B)
  n42 = Time()
  n43 = Add(
   A=Multiply(A=@n37.R, B=ScalarParameter() {group=TileNoise, parameterName=TileNoise Intensity})
   B=Add(A=Multiply(A=@n38, B=@n40), B=Multiply(A=@n41, B=@n42))
  )
  n44 = Multiply(A=ScalarParameter() {parameterName=Twitch}) {constB=0.001}
  n45 = TextureSampleParameter2D(UVs=Add(A=@n43, B=@n44)) {
   group=Tile
   parameterName=Tile Tex
   samplerType=SAMPLERTYPE_LinearGrayscale
   texture=/Game/Weapons/FORT_Melee/Pickaxe_Cube_Ninja_Male/FX/T_FluidFlameTile
  }
  n46 = TextureSampleParameter2D(UVs=@n43) {
   group=Tile
   parameterName=Tile Tex
   samplerType=SAMPLERTYPE_LinearGrayscale
   texture=&T_FluidFlameTile
  }
  n47 = TextureSampleParameter2D(UVs=Subtract(A=@n43, B=@n44)) {
   group=Tile
   parameterName=Tile Tex
   samplerType=SAMPLERTYPE_LinearGrayscale
   texture=&T_FluidFlameTile
  }
  n48 = MaterialFunctionCall(X=@n45.R, Y=@n46.R, Z=@n47.R) {materialFunction=&MakeFloat3}
  n49 = VectorParameter() {group=Tile, parameterName=Tile Refine, defaultValue=<rgba 1 1 1 0.5>}
  n50 = Multiply(A=Power(Base=@n48.Result, Exp=@n49.R), B=@n49.G)
  n51 = Multiply(A=@n48.Result, B=@n49.B)
  n52 = VectorParameter() {group=Emissive, parameterName=Emissive Refine, defaultValue=<rgba 1 1 1 0.5>}
  n53 = Power(Base=@n19, Exp=@n52.R)
  n54 = Add(
   A=Desaturation(None=LinearInterpolate(A=@n50, B=@n51, Alpha=@n49.A), Fraction=ScalarParameter() {parameterName=Saturation})
   B=LinearInterpolate(A=Multiply(A=@n53, B=@n52.G), B=Multiply(A=@n19, B=@n52.B), Alpha=@n52.A)
  )
  MaterialFunctionCall(Emissive=Multiply(A=Multiply(A=@n31, B=@n54), B=@n1.RGB)) {materialFunction=&MF_FX_EmissiveAdjustments}
  n55 = DepthFade(Opacity=@n27, FadeDistance=ScalarParameter() {parameterName=DepthFade})
  ShadingPathSwitch(Default=@n55, Deferred=@n55, Forward=@n27, Mobile=@n27)
 ]
}
