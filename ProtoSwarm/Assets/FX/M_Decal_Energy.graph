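# Translucent unlit deferred decal material for a swirling energy effect, colored by a hue-shifted curve-atlas gradient over the decal color.
# Scrolling polar-UV noise warps a flame tile, sampled thrice for RGB-split twitch, and an emissive gradient masked to a ring.
# Opacity combines that ring with a swirl-distorted sphere mask.
# Named reroutes provide an actor-distance 3D mask and pulsing world-space Voronoi waves, blending two phase-offset samples.
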Material /ProtoSwarm/Assets/FX/M_Decal_Energy
graph {
 properties {
  bUseTranslucencyVertexFog=false
  bUsedWithNiagaraMeshParticles=true
  blendMode=BLEND_Translucent
  materialDomain=MD_DeferredDecal
  pixelDepthOffsetMode=PDOM_Legacy
  refractionMethod=RM_IndexOfRefraction
  shadingModel=MSM_Unlit
  usedShadingModels=Unlit
 }
 nodes [
  n1 = DecalColor()
  n2 = ScalarParameter() {parameterName=HueShift}
  n3 = TextureCoordinate()
  n4 = CurveAtlasRowParameter(CurveTime=ComponentMask(None=MaterialFunctionCall!(UV Coord=@n3, LenghtScale, RadialScale) {materialFunction=/Game/Global/MaterialFunctions/TechArt/MF_PolarCoordinates}.Result) {r=true}) {atlas=&/CurveAtlas_DarkEnergy, curve=&/Curve_DarkEnergy_Color, parameterName=Curve}
  n5 = MaterialFunctionCall(UV Coord=TextureCoordinate()) {materialFunction=&MF_PolarCoordinates}
  n6 = VectorParameter() {group=TileNoise, parameterName=TileNoise UV, defaultValue=<rgba 1 1 0 0.25>}
  n7 = AppendVector(A=@n6.R, B=@n6.R)
  n8 = AppendVector(A=@n6.B, B=@n6.B)
  n9 = Time()
  n10 = TextureSampleParameter2D(UVs=Add(A=Multiply(A=@n5.Result, B=@n7), B=Multiply(A=@n8, B=@n9))) {
   group=TileNoise
   parameterName=TileNoise Tex
   samplerType=SAMPLERTYPE_Grayscale
   texture=&LowResBlurredNoise
  }
  n11 = MaterialFunctionCall(UV Coord=TextureCoordinate()) {materialFunction=&MF_PolarCoordinates}
  n12 = VectorParameter() {group=Tile, parameterName=Tile UV, defaultValue=<rgba 1 1 0 0.25>}
  n13 = AppendVector(A=@n12.R, B=@n12.R)
  n14 = AppendVector(A=@n12.B, B=@n12.B)
  n15 = Time()
  n16 = Add(
   A=Multiply(A=@n10.R, B=ScalarParameter() {group=TileNoise, parameterName=TileNoise Intensity})
   B=Add(A=Multiply(A=@n11.Result, B=@n13), B=Multiply(A=@n14, B=@n15))
  )
  n17 = Multiply(A=ScalarParameter() {parameterName=Twitch}) {constB=0.001}
  n18 = TextureSampleParameter2D(UVs=Add(A=@n16, B=@n17)) {
   group=Tile
   parameterName=Tile Tex
   samplerType=SAMPLERTYPE_LinearGrayscale
   texture=/Game/Weapons/FORT_Melee/Pickaxe_Cube_Ninja_Male/FX/T_FluidFlameTile
  }
  n19 = TextureSampleParameter2D(UVs=@n16) {
   group=Tile
   parameterName=Tile Tex
   samplerType=SAMPLERTYPE_LinearGrayscale
   texture=&T_FluidFlameTile
  }
  n20 = TextureSampleParameter2D(UVs=Subtract(A=@n16, B=@n17)) {
   group=Tile
   parameterName=Tile Tex
   samplerType=SAMPLERTYPE_LinearGrayscale
   texture=&T_FluidFlameTile
  }
  n21 = MaterialFunctionCall(X=@n18.R, Y=@n19.R, Z=@n20.R) {materialFunction=&MakeFloat3}
  n22 = VectorParameter() {group=Tile, parameterName=Tile Refine, defaultValue=<rgba 1 1 1 0.5>}
  n23 = Power(Base=@n21.Result, Exp=@n22.R)
  n24 = Desaturation(
   None=LinearInterpolate(A=Multiply(A=@n23, B=@n22.G), B=Multiply(A=@n21.Result, B=@n22.B), Alpha=@n22.A)
   Fraction=ScalarParameter() {parameterName=Saturation}
  )
  n25 = MaterialFunctionCall(UV Coord=TextureCoordinate()) {materialFunction=&MF_PolarCoordinates}
  n26 = VectorParameter() {group=EmissiveNoise, parameterName=EmissiveNoise UV, defaultValue=<rgba 1 1 0 0.25>}
  n27 = AppendVector(A=@n26.R, B=@n26.R)
  n28 = AppendVector(A=@n26.A, B=@n26.A)
  n29 = Time()
  n30 = TextureSampleParameter2D(UVs=Add(A=Multiply(A=@n25.Result, B=@n27), B=Multiply(A=@n28, B=@n29))) {
   group=EmissiveNoise
   parameterName=EmissiveNoise Tex
   samplerType=SAMPLERTYPE_Grayscale
   texture=/Game/Weapons/FORT_Melee/Pickaxe_Cube_Ninja_Male/FX/T_CellNoise
  }
  n31 = ScalarParameter() {defaultValue=1, group=EmissiveNoise, parameterName=EmissiveNoise Intensity}
  n32 = Multiply(A=OneMinus(None=@n30.R), B=@n31)
  n33 = Subtract(A=@n32, B=Multiply(A=@n30.R, B=@n31))
  n34 = MaterialFunctionCall(UV Coord=TextureCoordinate()) {materialFunction=&MF_PolarCoordinates}
  n35 = VectorParameter() {group=Emissive, parameterName=Emissive UV, defaultValue=<rgba 1 1 0 0.5>}
  n36 = AppendVector(A=@n35.R, B=@n35.R)
  n37 = ScalarParameter() {defaultValue=-1, parameterName=UVOffset}
  n38 = Add(
   A=Multiply(A=Constant2Vector() {g=1}, B=Multiply(A=@n32, B=@n33))
   B=Add(A=Multiply(A=@n34.Result, B=@n36), B=AppendVector(A=@n37, B=@n35.B))
  )
  n39 = Multiply(
   A=MaterialFunctionCall!(TexCoord=@n38, ClampChannel=Constant2Vector() {r=1}) {materialFunction=/Game/Effects/Fort_Effects/MaterialFunctions/MF_FX_TilingMethod}.Result
   B=ChannelMaskParameter(None=TextureSampleParameter2D(UVs=@n38) {group=Emissive, parameterName=Emissive Tex, texture=/Game/Weapons/FORT_Melee/Pickaxe_Cube_Ninja_Male/FX/T_Offset_Gradient_01}.RGBA) {group=Emissive, maskChannel=Blue, parameterName=Emissive Channel, defaultValue=<rgba 0 0 1 0>}.R
  )
  n40 = VectorParameter() {group=Emissive, parameterName=Emissive Refine, defaultValue=<rgba 1 1 1 0.5>}
  n41 = LinearInterpolate(A=Multiply(A=Power(Base=@n39, Exp=@n40.R), B=@n40.G), B=Multiply(A=@n39, B=@n40.B), Alpha=@n40.A)
  n42 = ComponentMask(None=MaterialFunctionCall(UV Coord=TextureCoordinate()) {materialFunction=&MF_PolarCoordinates}.Result) {r=true}
  n43 = Multiply(A=@n42, B=@n42)
  n44 = OneMinus(None=@n42)
  n45 = Multiply(A=@n44, B=@n44)
  n46 = Multiply(A=Multiply(A=Power(Base=@n43) {constExponent=1.2}, B=Power(Base=@n45) {constExponent=8})) {constB=20}
  n47 = Multiply(
   A=Multiply(A=@n39, B=ScalarParameter() {defaultValue=1, group=Emissive, parameterName=Emissive Mask})
   B=Multiply(A=Saturate(None=@n46), B=ScalarParameter() {defaultValue=20, parameterName=Mask Multiply}) {constB=10}
  )
  LinearInterpolate(
   A=@n1.RGB
   B=Multiply(
    A=MaterialFunctionCall(Hue Shift Percentage=@n2, Texture=@n4) {materialFunction=&HueShift}.Result
    B=Add(A=@n24, B=@n41)
   )
   Alpha=@n47
  )
  MaterialFunctionCall(Emissive=NamedRerouteUsage()) {materialFunction=&MF_FX_EmissiveAdjustments}
  n48 = MaterialFunctionCall() {materialFunction=&VectorToRadialValue}
  n49 = Constant2Vector() {g=1.5, r=4}
  n50 = TextureSample(UVs=Panner(Coordinate=Multiply(A=@n48.Radial Coordinates, B=@n49)) {bFractionalPart=true, speedY=-0.55}) {samplerType=SAMPLERTYPE_Normal, texture=/Game/Global/Textures/Default/Noise/T_SwirlNoise_N}
  n51 = Add(
   A=TextureCoordinate()
   B=Multiply(A=ComponentMask(None=@n50.RGB) {g=true, r=true}, B=ScalarParameter() {defaultValue=0.2, parameterName=DistortScale})
  ) {constB=0}
  n52 = Constant() {r=0.5}
  n53 = ScalarParameter() {defaultValue=0.45, parameterName=Radius}
  n54 = ScalarParameter() {defaultValue=0.9, parameterName=Hardness}
  n55 = DotProduct(A=Constant3Vector() {constant=<rgba 0.5 0.5 0 0>}, B=@n50.RGB)
  n56 = Saturate(None=Multiply(A=SphereMask(A=@n51, B=@n52, Radius=@n53, Hardness=@n54), B=Add(A=@n55)))
  Multiply(A=Add(A=Multiply(A=@n1.A, B=@n56), B=@n47), B=NamedRerouteUsage())
  n57 = ScalarParameter() {defaultValue=0.001, group=Mask, parameterName=CenterMask Adjustment}
  n58 = Distance(A=Subtract(A=ActorPositionWS(), B=WorldPosition().XYZ), B=Constant3Vector())
  n59 = Saturate(None=Multiply(A=Multiply(A=@n57, B=@n58) {constB=0.05}, B=ScalarParameter() {defaultValue=3.5, parameterName=3DMask_Adjustment}) {constB=5})
  NamedRerouteDeclaration(None=Power(Base=OneMinus(None=@n59), Exp=ScalarParameter() {defaultValue=0.4, parameterName=3DMaskPow})) {name=3DMask, nodeColor=<rgba 1 0.6328 0 1>}
  n60 = WorldPosition()
  n61 = Distance(A=Subtract(A=ActorPositionWS(), B=@n60.XYZ), B=VectorParameter() {parameterName=Ofs, defaultValue=<rgba 20 0 0 1>}.RGB)
  n62 = ScalarParameter() {defaultValue=0.0007, parameterName=Distance Adjustment2}
  n63 = Multiply(A=Time())
  n64 = Frac(None=Add(A=@n63) {constB=0.5})
  n65 = Multiply(A=@n61, B=Add(B=Multiply(A=@n62, B=@n64)) {constA=0.0006}) {constB=0.01}
  n66 = Multiply(A=Add(A=Sine(None=@n65))) {constB=0.5}
  n67 = ActorPositionWS()
  n68 = Multiply(
   A=Multiply(A=Power(Base=@n66) {constExponent=-2.5}, B=Subtract(A=@n67, B=@n60.XYZ))
   B=ScalarParameter() {defaultValue=0.007, group=Volume Noise, parameterName=UVW_Scale2}
  )
  n69 = Rotator(Coordinate=ComponentMask(None=@n68) {g=true, r=true}, Time=Time())
  n70 = ComponentMask(None=@n68) {b=true}
  n71 = Subtract(A=TextureSample(UVs=AppendVector(A=@n69, B=@n70)) {samplerType=SAMPLERTYPE_LinearColor, texture=/Game/Effects/Explosion_Update/Textures/VT_Voronoi_Seamless}.RGB) {constB=0.1}
  n72 = Cosine(None=Add(A=@n63) {constB=0.1})
  n73 = Multiply(A=Power(Base=Saturate(None=@n71)) {constExponent=8}, B=Add(A=Multiply(A=@n72) {constB=0.5}) {constB=0.5})
  n74 = Add(B=Multiply(A=Frac(None=@n63), B=ScalarParameter() {defaultValue=0.0005, parameterName=Distance Adjustment}) {constB=0.01}) {constA=0.0007}
  n75 = Add(A=Sine(None=Multiply(A=@n61, B=@n74) {constB=0.01}))
  n76 = Multiply(A=Power(Base=Multiply(A=@n75) {constB=0.5}) {constExponent=-2.5}, B=Subtract(A=ActorPositionWS(), B=@n60.XYZ))
  n77 = ScalarParameter() {defaultValue=0.008, group=Volume Noise, parameterName=UVW_Scale}
  n78 = Subtract(A=TextureSample(UVs=Multiply(A=@n76, B=@n77)) {samplerType=SAMPLERTYPE_LinearColor, texture=&VT_Voronoi_Seamless}.RGB) {constB=0.1}
  n79 = Sine(None=Add(A=@n63) {constB=0.8})
  n80 = Multiply(A=Power(Base=Saturate(None=@n78)) {constExponent=8}, B=Add(A=Multiply(A=@n79) {constB=0.5}) {constB=0.5})
  NamedRerouteDeclaration(None=Multiply(A=Add(A=@n73, B=@n80), B=VectorParameter() {parameterName=3D Noise Wave Color, defaultValue=<rgba 6.8373 2.1236 15 1>}.RGB)) {name=3D Noise Wave, nodeColor=<rgba 0.5 1 0 1>}
 ]
}
