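# Masked two-sided material for Niagara mesh particles, apparently the digitized cubes of an elimination effect.
# Emissive combines a colored grid with HDR Fresnel glow, driven by particle color and dynamic parameters, with a simplified ray-tracing branch for Lumen.
# Particle alpha drives dissolve, either by instance-height erosion broken up with a crack texture or a UV-based fallback, with dithered near-camera fade.
# An optional switch shrinks particles near the camera through world position offset.
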Material /ProtoSwarm/Assets/FX/M_Elimination_DigitizeCubes_TacElim
graph {
 properties {
  bEnableResponsiveAA=true
  bUsedWithNiagaraMeshParticles=true
  blendMode=BLEND_Masked
  pixelDepthOffsetMode=PDOM_Legacy
  translucencyLightingMode=TLM_VolumetricDirectional
  twoSided=true
  usedShadingModels=Default Lit
 }
 nodes [
  ActorPositionWS()
  ActorPositionWS()
  n1 = DynamicParameter() {defaultValue=<rgba 0 0 1 0>, paramNames [FresnelGlow, UseParticleColor, GridBrightness, CubeLerp]}
  n2 = Saturate(None=@n1.CubeLerp)
  LinearInterpolate(A=Constant() {r=0.8}, Alpha=@n2)
  LinearInterpolate(A=Constant() {r=0.3}, Alpha=@n2) {constB=0.125}
  n3 = DynamicParameter() {defaultValue=<rgba 0 0 1 0>, paramNames [FresnelGlow, UseParticleColor, GridBrightness, CubeLerp]}
  n4 = LinearInterpolate(
   A=VectorParameter() {group=CINEMATICS- RezOut, parameterName=GridColor, defaultValue=<rgba 0.6748 0 1 1>}.RGB
   B=ParticleColor().RGB
   Alpha=@n3.UseParticleColor
  )
  n5 = Multiply(
   A=Multiply(A=@n4, B=ScalarParameter() {defaultValue=1.5, group=CINEMATICS- RezOut, parameterName=CINE_DeRezGridBrightness}) {constB=2}
   B=@n3.GridBrightness
  )
  n6 = Add(A=Multiply(A=@n5, B=NamedRerouteUsage()), B=NamedRerouteUsage())
  n7 = Multiply(A=@n5, B=ScalarParameter() {defaultValue=5, parameterName=Lumenintensity})
  MaterialFunctionCall(Emissive=StaticSwitchParameter(True=RayTracingQualitySwitch(Normal=@n6, RayTraced=@n7), False=Constant()) {defaultValue=true, parameterName=UseEmissive}) {materialFunction=&MF_FX_EmissiveAdjustments}
  MaterialFunctionCall() {materialFunction=&DebugTimeSine}
  MaterialFunctionCall!(Center, Height=Constant() {r=1}) {materialFunction=/Game/Effects/Fort_Effects/Materials/_Functions/MF_Z_GradientFromObjectHeight_01}
  Multiply()
  n8 = ParticleColor()
  n9 = Add(
   A=ComponentMask(None=TextureCoordinate() {coordinateIndex=1}) {r=true}
   B=LinearInterpolate(Alpha=NamedRerouteDeclaration(None=@n8.A) {name=Particle Alpha, nodeColor=<rgba 0 0.0156 1 1>}) {constA=-1}
  )
  n10 = Constant() {r=28}
  Multiply(
   A=StaticSwitchParameter(True=NamedRerouteUsage(), False=Clamp(None=@n9)) {defaultValue=true, parameterName=UseZHeightErosion}
   B=MaterialFunctionCall(Alpha Threshold=MaterialFunctionCall(Distance=@n10) {materialFunction=&Near_Camera_Fade}.Result) {materialFunction=&DitherTemporalAA_Fortnite}.Result
  )
  n11 = MaterialFunctionCall() {materialFunction=/Game/Effects/Niagara/MasterMaterials/Functions/InstanceBounds_0-1_UVW}
  n12 = NamedRerouteUsage()
  n13 = Multiply(A=Subtract(A=Power(Base=@n11.B) {constExponent=0.75}, B=LinearInterpolate(Alpha=@n12) {constA=-1, constB=1.5}) {constB=1.5}) {constB=1.25}
  n14 = Power(Base=TextureSample(UVs=TextureCoordinate() {uTiling=0.25, vTiling=0.25}) {samplerType=SAMPLERTYPE_LinearColor, texture=/Game/Abilities/Player/Outlander/Actives/GroundWave/FX/T_GridCrack_Mask}.B)
  n15 = Subtract(A=OneMinus(None=Saturate(None=@n13)), B=Saturate(None=Subtract(A=@n14, B=@n12)))
  NamedRerouteDeclaration(None=@n15) {name=Erosion, nodeColor=<rgba 0.2891 1 0 1>}
  n16 = Multiply(A=TextureCoordinate(), B=ScalarParameter() {defaultValue=0.25, group=CINEMATICS- RezOut, parameterName=CINE_DeRezGridSize})
  NamedRerouteDeclaration(None=ComponentMask(None=TextureSample(UVs=@n16) {samplerType=SAMPLERTYPE_LinearGrayscale, texture=/Game/Packages/Fortress_SharedMaterials/Textures/T_Grid__M}.RGB) {r=true}) {name=Grid, nodeColor=<rgba 1 0.2812 0 1>}
  n17 = LinearInterpolate(A=VectorParameter() {parameterName=Color, defaultValue=<rgba 0.5224 0.3185 1 1>}.RGB, B=@n8.RGB, Alpha=@n1.UseParticleColor)
  n18 = MaterialFunctionCall(Brightness=@n3.FresnelGlow, Exponent=Constant() {r=2}) {materialFunction=&MF_Fresnel_FX}
  NamedRerouteDeclaration(None=MaterialFunctionCall(LDR Input=Multiply(A=@n17, B=@n18.Fresnel), HDR Dynamic Range=Constant() {r=1}, HDR Tint=@n4, HDR Intensity Multiplier=Constant() {r=100}) {materialFunction=&DeriveHDRfromLDR}.RemappedRange) {name=Fresnel, nodeColor=<rgba 1 0 0.6797 1>}
  n19 = Subtract(A=ParticlePositionWS(), B=WorldPosition().XYZ)
  n20 = NamedRerouteUsage()
  NamedRerouteDeclaration(None=StaticSwitchParameter(True=LinearInterpolate(A=@n19, Alpha=@n20) {constAlpha=0, constB=0}, False=Constant()) {parameterName=UseNearCameraShrink}) {name=To WPO, nodeColor=<rgba 0.6172 0 1 1>}
  n21 = Length(None=Subtract(A=CameraPositionWS(), B=ParticlePositionWS())) {is=class MaterialExpressionLength}
  n22 = ScalarParameter() {defaultValue=25, desc=Near distance to camera --250, group=CamFade, parameterName=CamFade Near, sliderMax=250, sortPriority=1}
  n23 = ScalarParameter() {defaultValue=45, desc=Far distance to camera -- 300, group=CamFade, parameterName=CamFade Far, sliderMax=300, sortPriority=2}
  NamedRerouteDeclaration(None=Saturate(None=MaterialFunctionCall(Input=@n21, Input Low=@n22, Input High=@n23) {materialFunction=&RemapValue_0-1}.Result)) {name=ClipFade Alpha, nodeColor=<rgba 0 1 0.9447 1>}
  NamedRerouteUsage()
  ObjectLocalBounds()
  Saturate(None=@n15)
 ]
}
