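# Unlit material for goo drop effects, flagged for use on Niagara mesh particles.
# An inverted, sharpened Fresnel makes a faint, tight highlight where the surface faces the camera, with no lighting involved.
# That highlight is added to the particle color, so the emitter controls the tint, and the sum feeds the shared MF_FX_EmissiveAdjustments function as its emissive input.
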Material /ProtoSwarm/Assets/FX/M_Goo_Drop
graph {
 properties {
  bTangentSpaceNormal=false
  bUsedWithNiagaraMeshParticles=true
  pixelDepthOffsetMode=PDOM_Legacy
  shadingModel=MSM_Unlit
  usedShadingModels=Unlit
 }
 nodes [
  n1 = Fresnel() {exponent=0.5}
  n2 = Saturate(None=Power(Base=OneMinus(None=@n1)) {constExponent=9})
  MaterialFunctionCall(Emissive=Add(A=Multiply(A=@n2) {constB=0.3}, B=ParticleColor().RGB)) {materialFunction=&MF_FX_EmissiveAdjustments}
 ]
}
