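# Subsurface material for TidalCrane projectile goo on Niagara sprite and mesh particles, with animatable shape and color through dynamic parameters.
# Actor-relative value noise, or a center-scaled texture, is multiplied and wrapped with Frac into repeating bands that displace vertices along normals.
# The same bands, optionally posterized with Floor, blend three colors added to particle color.
# Fresnel terms shape edge emissive, subsurface color and inner-to-outer opacity.
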Material /ProtoSwarm/Assets/FX/M_TidalCrane_Prj_Goo
graph {
 properties {
  bUsedWithNiagaraMeshParticles=true
  bUsedWithNiagaraSprites=true
  pixelDepthOffsetMode=PDOM_Legacy
  shadingModel=MSM_Subsurface
  usedShadingModels=Subsurface
 }
 nodes [
  n1 = DynamicParameter() {parameterIndex=1, defaultValue=<rgba 1 0 1 0>, paramNames [WPOMult, UseWPOTexture, ScaveUVTexWPO, Empty3]}
  n2 = MaterialFunctionCall(UVs=TextureCoordinate(), Texture Scale=@n1.ScaveUVTexWPO) {materialFunction=&ScaleUVsByCenter}
  n3 = Subtract(A=WorldPosition().XYZ, B=ActorPositionWS())
  n4 = DynamicParameter() {defaultValue=<rgba 100 1 50 0>, paramNames [Param1, Param2, GradientSoftness, UVOffset]}
  n5 = Noise!(World Position=Divide(A=Add(A=@n3, B=@n4.UVOffset), B=@n4.Param1), FilterWidth) {bTurbulence=false, levels=1, noiseFunction=NOISEFUNCTION_ValueALU, outputMin=0}
  n6 = If(
   A=@n1.UseWPOTexture
   "A > B"=TextureSample(UVs=@n2.UVs) {texture=&T_Smooth_tile}.RGB
   "A == B"=@n5
   "A < B"=@n5
  )
  n7 = Frac(None=Multiply(A=ComponentMask(None=@n6) {g=true}, B=@n4.Param2))
  n8 = LinearInterpolate(Alpha=@n7) {constA=-1}
  n9 = VertexNormalWS()
  n10 = DynamicParameter() {parameterIndex=1, defaultValue=<rgba 1 0 1 0>, paramNames [WPOMult, UseWPOTexture, ScaveUVTexWPO, Empty3]}
  Add(A=Multiply(A=Multiply(A=@n8, B=@n9), B=@n10.WPOMult)) {constB=0}
  Constant() {r=0.2}
  n11 = DynamicParameter() {defaultValue=<rgba 100 1 50 0>, paramNames [Param1, Param2, GradientSoftness, UVOffset]}
  n12 = ParticleColor()
  n13 = VectorParameter() {parameterName=Color2, defaultValue=<rgba 0.3333 0.0138 0 0>}
  n14 = Floor(None=Multiply(A=@n7, B=@n4.GradientSoftness))
  n15 = MaterialFunctionCall(
   1=VectorParameter() {parameterName=Color1, defaultValue=<rgba 0.9375 0.288 0 0>}.RGB
   2=VectorParameter() {parameterName=Color3, defaultValue=<rgba 2 1.4545 0 0>}.RGB
   3=@n13.RGB
   4=@n13.RGB
   A=LinearInterpolate(A=@n7, B=Divide(A=@n14, B=@n4.GradientSoftness), Alpha=ScalarParameter() {parameterName=GradientSoftnessLerp}) {constAlpha=1}
  ) {materialFunction=&Lerp_Multiple_Float3}
  If(A=@n11.GradientSoftness, "A > B"=Add(A=@n12.RGB, B=@n15.Lerp 4 Inputs), "A == B"=@n12.RGB, "A < B"=@n12.RGB)
  LinearInterpolate(
   A=ScalarParameter() {defaultValue=0.8, parameterName=OpacityInner}
   B=ScalarParameter() {defaultValue=0.25, parameterName=OpacityOuter}
   Alpha=Fresnel(ExponentIn=ScalarParameter() {defaultValue=1, parameterName=OpacityFresnelExp}) {baseReflectFraction=0, exponent=1}
  ) {constA=0.5, constB=0.1}
  n16 = Fresnel(ExponentIn=ScalarParameter() {defaultValue=5, parameterName=EdgeFresnelExp}) {baseReflectFraction=0}
  n17 = If(A=@n11.GradientSoftness, "A > B"=Add(A=@n12.RGB, B=Multiply(A=@n15.Lerp 4 Inputs, B=@n16)), "A == B"=@n12.RGB, "A < B"=@n12.RGB)
  MaterialFunctionCall(Emissive=Multiply(A=@n17, B=ScalarParameter() {defaultValue=0.5, parameterName=EdgeEmissiveMult})) {materialFunction=&MF_FX_EmissiveAdjustments}
  Multiply(A=@n17, B=ScalarParameter() {defaultValue=1, parameterName=SubsurfaceColorMult}) {constB=0.5}
  Power(Base=@n15.Lerp 4 Inputs, Exp=ScalarParameter() {defaultValue=1, parameterName=GlowPower})
  TextureSample(UVs=@n2.UVs) {samplerType=SAMPLERTYPE_Grayscale, texture=&LowResBlurredNoise}
  TextureSample() {texture=&T_Inky_Smoke_Tile}
  VectorParameter() {parameterName=Color11, defaultValue=<rgba 0.1313 0.9375 0 0>}
  VectorParameter() {parameterName=Color22, defaultValue=<rgba 0.0253 0.3333 0 0>}
  VectorParameter() {parameterName=Color33, defaultValue=<rgba 2 0.9629 0 0>}
 ]
}
