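# Niagara crystal burst effect that throws out mesh shards with a layered flash, rays and sparks.
# Crystal and Crystal_Large shards fly outward, fall under gravity and flicker via brightness curves; only small shards collide.
# MainShape layers sprite flares and meshes by spawn group, Rays adds expanding rotating meshes, and Lines stretches repelled sprites by speed.
# User.Offset shifts shards, rays and lines; User.Size_Scale scales only MainShape's mesh.
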NiagaraSystem /ProtoSwarm/Assets/FX/NS_Proto_Crystal_Burst_01
dependencies {
 usedDataInterfaces [&NiagaraDataInterfaceCurve, &NiagaraDataInterfaceArrayFloat3, &NiagaraDataInterfaceColorCurve]
 usedDynamicInputs [&FNTimeOfDayScalar]
 usedModules [
  &SystemState
  &self:SystemSpawnScript.SystemScriptSource.SystemScriptGraph.NiagaraNodeAssignment_0.SetVariables_#1
 ]
}
emitters {
 Crystal {
  emitterUpdateScript [
   &EmitterState(Life Cycle Mode=::NewEnumerator1, Loop Behavior=::NewEnumerator1, Loop Duration=1, UseLoopDelay=true)
   &SpawnBurst_Instantaneous(Spawn Count=10, Use Spawn Probability=false, UseLoopCountLimit=false)
  ]
  particleSpawnScript [
   /SetVariables(Particles.SpawnRotationVector=&RandomRangeVector(Minimum=<-2 0 0>, Maximum=<2 0 0>)) {isSetParameters=true}
   &InitializeParticle(
    Lifetime=&RandomRangeFloat(Minimum=1, Maximum=1.35)
    Mesh Scale=&RandomRangeVector(Minimum=<0.2 0.2 0.2>, Maximum=<0.75 0.75 0.75>)
    Mesh Scale Mode=::NewEnumerator4
    UsePositionOffset=false
    Write Material Random=true
   )
   &Color(Color=&MakeLinearColorFromVectorAndFloat(
     "Vector (RGB)"=&VectorFromFloat(Value=&RandomRangeFloat(Minimum=0.2, Maximum=2))
     "Float (Alpha)"=&ClampFloat(Float=1)
    ))
   &SystemLocation(MaskSpawnGroup=false, Offset=$User.Offset)
   &ShapeLocation(
    Hemisphere Distribution=<1 0.5>
    MaskSpawnGroup=false
    Override Seed=false
    Sphere Radius=&RandomRangeFloat(Minimum=17.5, Maximum=35)
    UseOwnerScale=true
   )
   &AddVelocity(Velocity=&Multiply_VectorByFloat(Vector=$Particles.ShapeLocation.ShapeVector, Float=&RandomRangeFloat(Minimum=20, Maximum=60)))
   /Niagara/Modules/CascadeConversion/CascadeConversion_InitialMeshRotation(Rotation:Vector3f=&RandomRangeVector(Minimum=<-1 -1 -1>, Maximum=<1 1 1>))
   &AddVelocity(
    Module.Use Falloff Distance=false
    Origin Offset=<0 0 -10>
    Override Seed=false
    Velocity Mode=::NewEnumerator1
    Velocity Speed=&RandomRangeFloat(Minimum=50, Maximum=250)
   ) {module=AddVelocity001}
   &InheritVelocity(Inherited Velocity Amount Scale=<0.3 0.3 0.3>, Inherited Velocity Speed Limit=0, Speed Limit=false, UseSpeedThreshold=false)
  ]
  particleUpdateScript [
   &ParticleState()
   &ScaleMeshSize(Scale Factor=&VectorFromFloat(Value=&FloatFromCurve(FloatCurve=DataInterface() {propertyValues {Curve {keys [
          <at 0 0 RCIM_Cubic>
          <at 0.0344 1 RCIM_Constant>
          {arriveTangent=-1.0357, interpMode=RCIM_Cubic, leaveTangent=-1.0357, time=0.8718, value=0.9886}
          {interpMode=RCIM_Cubic, leaveTangent=-11.4404, time=1, value=0}
         ]}}})))
   &UpdateMeshOrientation(Rotation Vector=$Particles.SpawnRotationVector)
   &ScaleColor(
    Scale RGB=&VectorFromFloat(Value=&FloatFromCurve(FloatCurve=DataInterface() {propertyValues {Curve {keys [
          {arriveTangent=2.113, time=0, value=1}
          {arriveTangent=-45.2078, interpMode=RCIM_Cubic, leaveTangent=-45.2078, time=0.0151, value=3.113}
          {arriveTangent=-11.9511, interpMode=RCIM_Cubic, leaveTangent=-11.9511, time=0.0214, value=0.0304}
          {arriveTangent=1.6211, interpMode=RCIM_Cubic, leaveTangent=1.6211, time=0.0613, value=2.5609}
          {arriveTangent=-8.6158, interpMode=RCIM_Cubic, leaveTangent=-8.6158, time=0.1019, value=0.1608}
          {arriveTangent=8.1161, interpMode=RCIM_Cubic, leaveTangent=8.1161, time=0.1134, value=2.1118}
          {arriveTangent=-1.4429, interpMode=RCIM_Cubic, leaveTangent=-1.4429, time=0.158, value=0.616}
          {arriveTangent=-2.2029, interpMode=RCIM_Cubic, leaveTangent=-2.2029, time=0.1963, value=1.9923}
          {arriveTangent=-10.0958, interpMode=RCIM_Cubic, leaveTangent=-10.0958, time=0.2443, value=0.4259}
          {arriveTangent=-0.5636, interpMode=RCIM_Cubic, leaveTangent=-0.5636, time=0.3846, value=0.0909}
          <at 1 0>
         ]}}}))
    ScaleA=true
    ScaleRGB=true
   )
   &Drag(
    Drag=&FloatFromCurve(FloatCurve=DataInterface() {propertyValues {Curve {keys [<at 0 2>,<at 0.1 0>]}}})
    Use Linear Drag=true
    Use Rotational Drag=true
   )
   &AccelerationForce(Acceleration=<0 0 -1200>)
   &Collision(
    Age Colliding Particles=false
    KillOnCollision=false
    Module.Control Rest State Via Surface Normal=false
    Module.EnableMaxCollisionCount=false
    Module.Randomize Collision Normal Vector=false
    Restitution=0.25
   )
   &SolveForcesAndVelocity(Clamp Velocity=false, Limit Acceleration=false)
   &CameraOffset(Camera Offset Amount=60)
  ]
  rendererClasses [&NiagaraMeshRendererProperties]
  renderers [{rendererClass=&NiagaraMeshRendererProperties, rendererIndex=0}]
 }
 Crystal_Large {
  emitterUpdateScript [
   &EmitterState(Life Cycle Mode=::NewEnumerator1, Loop Behavior=::NewEnumerator1, Loop Duration=1, UseLoopDelay=true)
   &SpawnBurst_Instantaneous(Spawn Count=4, Use Spawn Probability=false, UseLoopCountLimit=false)
  ]
  particleSpawnScript [
   /SetVariables(Particles.SpawnRotationVector=&RandomRangeVector(Minimum=<-2 0 0>, Maximum=<2 0 0>)) {isSetParameters=true}
   &InitializeParticle(
    Color=<rgba 1 0.6803 0 1>
    Lifetime=&RandomRangeFloat(Minimum=0.7, Maximum=1.5)
    Mesh Scale=&RandomRangeVector(Minimum=<0.9 0.9 0.9>, Maximum=<1.5 1.5 1.5>)
    Mesh Scale Mode=::NewEnumerator4
    UsePositionOffset=false
    Write Material Random=true
   )
   &Color(Color=&MakeLinearColorFromVectorAndFloat(
     "Vector (RGB)"=&VectorFromFloat(Value=&RandomRangeFloat(Minimum=0.2, Maximum=2))
     "Float (Alpha)"=&ClampFloat(Float=1)
    )) {enabled=false}
   &SystemLocation(MaskSpawnGroup=false, Offset=$User.Offset)
   &ShapeLocation(
    Hemisphere Distribution=<1 0.5>
    MaskSpawnGroup=false
    Override Seed=false
    Sphere Radius=&RandomRangeFloat(Minimum=17.5, Maximum=35)
    UseOwnerScale=true
   )
   &AddVelocity(Velocity=&Multiply_VectorByFloat(Vector=$Particles.ShapeLocation.ShapeVector, Float=&RandomRangeFloat(Minimum=6, Maximum=10)))
   &CascadeConversion_InitialMeshRotation(Rotation:Vector3f=&RandomRangeVector(Minimum=<-1 -1 -1>, Maximum=<1 1 1>))
   &AddVelocity(Velocity=&RandomRangeVector(Minimum=<0 0 50>, Maximum=<0 0 250>)) {module=AddVelocity001}
   &AddVelocity(
    Module.Use Falloff Distance=false
    Origin Offset=<0 0 -20>
    Override Seed=false
    Velocity Mode=::NewEnumerator1
    Velocity Speed=&RandomRangeFloat(Minimum=250, Maximum=650)
   ) {module=AddVelocity002}
   &InheritVelocity(Inherited Velocity Amount Scale=<0.3 0.3 0.3>, Inherited Velocity Speed Limit=0, Speed Limit=false, UseSpeedThreshold=false)
  ]
  particleUpdateScript [
   &ParticleState()
   &UpdateMeshOrientation(Rotation Vector=$Particles.SpawnRotationVector)
   &ScaleMeshSize(Scale Factor=&VectorFromFloat(Value=&FloatFromCurve(FloatCurve=DataInterface() {propertyValues {Curve {keys [
          <at 0 0 RCIM_Cubic>
          <at 0.0344 1 RCIM_Constant>
          {arriveTangent=-1.0357, interpMode=RCIM_Cubic, leaveTangent=-1.0357, time=0.8718, value=0.9886}
          {interpMode=RCIM_Cubic, leaveTangent=-11.4404, time=1, value=0}
         ]}}})))
   &ScaleColor(
    Scale RGB=&VectorFromFloat(Value=&FloatFromCurve(FloatCurve=DataInterface() {propertyValues {Curve {keys [
          {arriveTangent=2.113, time=0, value=1}
          {arriveTangent=-45.2078, interpMode=RCIM_Cubic, leaveTangent=-45.2078, time=0.0151, value=3.113}
          {arriveTangent=-11.9511, interpMode=RCIM_Cubic, leaveTangent=-11.9511, time=0.0214, value=0.0304}
          {arriveTangent=1.6211, interpMode=RCIM_Cubic, leaveTangent=1.6211, time=0.0613, value=2.5609}
          {arriveTangent=-8.6158, interpMode=RCIM_Cubic, leaveTangent=-8.6158, time=0.1019, value=0.1608}
          {arriveTangent=8.1161, interpMode=RCIM_Cubic, leaveTangent=8.1161, time=0.1134, value=2.1118}
          {arriveTangent=-1.4429, interpMode=RCIM_Cubic, leaveTangent=-1.4429, time=0.158, value=0.616}
          {arriveTangent=-2.2029, interpMode=RCIM_Cubic, leaveTangent=-2.2029, time=0.1963, value=1.9923}
          {arriveTangent=-10.0958, interpMode=RCIM_Cubic, leaveTangent=-10.0958, time=0.2443, value=0.4259}
          {arriveTangent=-0.5636, interpMode=RCIM_Cubic, leaveTangent=-0.5636, time=0.3846, value=0.0909}
          <at 1 0>
         ]}}}))
    ScaleA=true
    ScaleRGB=true
   )
   &Drag(
    Drag=&FloatFromCurve(FloatCurve=DataInterface() {propertyValues {Curve {keys [<at 0 2>,<at 0.1 0>]}}})
    Use Linear Drag=true
    Use Rotational Drag=true
   )
   &AccelerationForce(Acceleration=<0 0 -1200>)
   &SolveForcesAndVelocity(Clamp Velocity=false, Limit Acceleration=false)
   &GenerateLocationEvent(Use Event Probability=false, UseEventDelay=false) {module=GenerateLocationEvent001}
   &CameraOffset(Camera Offset Amount=60)
  ]
  rendererClasses [&NiagaraMeshRendererProperties]
  renderers [{rendererClass=&NiagaraMeshRendererProperties, rendererIndex=0}]
 }
 Lines {
  emitterUpdateScript [
   &EmitterState(Life Cycle Mode=::NewEnumerator1, Loop Behavior=::NewEnumerator1, Loop Duration=1, UseLoopDelay=true)
   &SpawnBurst_Instantaneous(Spawn Count=35, Use Spawn Probability=false, UseLoopCountLimit=false)
  ]
  particleSpawnScript [
   /SetVariables(Particles.SpriteRotationRate=&RandomRangeFloat(Minimum=-1, Maximum=1)) {isSetParameters=true}
   &InitializeParticle(
    Lifetime Max=0.5
    Lifetime Min=0.25
    Lifetime Mode=::NewEnumerator1
    Sprite Size Max=<6 9>
    Sprite Size Min=<2 3>
    Sprite Size Mode=::NewEnumerator2
    UsePositionOffset=false
    Write Material Random=true
   )
   &Color(Color=<rgba 128 0 60 1>)
   &SystemLocation(MaskSpawnGroup=false, Offset=$User.Offset)
   &ShapeLocation(MaskSpawnGroup=false, Offset Mode=::NewEnumerator0, Override Seed=false, UseOwnerScale=true)
   &AddVelocity(Velocity=&Multiply_VectorByFloat(Vector=$Particles.ShapeLocation.ShapeVector, Float=&RandomRangeFloat(Minimum=5, Maximum=20)))
   &InheritVelocity(Inherited Velocity Amount Scale=<0.5 0.5 0.5>, Inherited Velocity Speed Limit=0, Speed Limit=false, UseSpeedThreshold=false)
  ]
  particleUpdateScript [
   &ParticleState()
   &Color(Color=&ColorFromCurve(ColorCurve=DataInterface() {propertyValues {
       AlphaCurve {keys [<at 0 1 RCIM_Cubic>]}
       BlueCurve {keys [<at 0.2865 0 RCIM_Cubic>,<at 0.6205 60 RCIM_Cubic>]}
       GreenCurve {keys [<at 0.2865 50 RCIM_Cubic>,<at 0.6205 0 RCIM_Cubic>]}
       RedCurve {keys [<at 0.2865 100 RCIM_Cubic>,<at 0.6205 128 RCIM_Cubic>]}
      }})) {module=Color001}
   &SpriteRotationRate(Rotation Rate=&AngleConversion(
     Angle:NiagaraFloat=$Particles.SpriteRotationRate
     Angle Input:ENiagara_AngleInput=::NewEnumerator1
     Angle Output:ENiagara_AngleInput=::NewEnumerator0
    ))
   &ScaleSpriteSize(
    Scale Factor=&Vector2DFromFloat(Value=&FloatFromCurve(
      FloatCurve=DataInterface() {propertyValues {Curve {keys [
          {arriveTangent=-2.1021, time=-0.0044, value=1}
          {arriveTangent=-0.9826, interpMode=RCIM_Cubic, leaveTangent=-0.9826, time=0.157, value=0.2002}
          {arriveTangent=-0.2375, interpMode=RCIM_Cubic, leaveTangent=-0.2375, time=0.7034, value=0.3045}
          <at 1 0>
         ]}}}
      Scale Curve=2.6282
     ))
    Scale Sprite Size Mode=::NewEnumerator2
   )
   &Drag(
    Drag=&FloatFromCurve(FloatCurve=DataInterface() {propertyValues {Curve {keys [<at 0 1>,<at 1 4>]}}})
    Use Linear Drag=true
    Use Rotational Drag=true
   )
   /Niagara/Modules/Update/Forces/PointAttractionForce(
    Attraction Radius:NiagaraFloat=&RandomRangeFloat(Minimum=0, Maximum=1000, Evaluation Type=::NewEnumerator1)
    AttractionStrength=-500
    AttractorPosition:Vector3f=&RandomRangeVector(Minimum=<-100 -100 -100>, Maximum=<100 100 100>, Evaluation Type=::NewEnumerator1)
    Module.Kill Within Radius=false
    Module.Use Falloff=true
   )
   &SolveForcesAndVelocity(Clamp Velocity=false, Limit Acceleration=false)
   &ScaleSpriteSizeBySpeed(
    Max Scale Factor=<5 25>
    Min Scale Factor=<0 0>
    Sample Scale Factor By Curve=true
    Scale Factor Curve=DataInterface() {propertyValues {Curve {keys [
        {interpMode=RCIM_Cubic, tangentMode=RCTM_User, time=0, value=0}
        {interpMode=RCIM_Cubic, tangentMode=RCTM_User, time=0.5, value=1}
        {interpMode=RCIM_Cubic, tangentMode=RCTM_User, time=1, value=0}
       ]}}}
    Source Velocity:Vector3f=&Multiply_Vector(A=<0 0.002 1>, B=$Particles.Velocity)
    Velocity Threshold=1200
   )
   &CameraOffset(Camera Offset Amount=100)
  ]
  rendererClasses [&NiagaraSpriteRendererProperties]
  renderers [{rendererClass=&NiagaraSpriteRendererProperties}]
 }
 MainShape {
  emitterUpdateScript [
   &EmitterState()
   &SpawnBurst_Instantaneous(Spawn Time=0.05, Use Spawn Probability=false, UseLoopCountLimit=false)
   &SpawnBurst_Instantaneous(Spawn Group=1, Spawn Time=0.075, Use Spawn Probability=false, UseLoopCountLimit=false) {module=SpawnBurst_Instantaneous001}
   &SpawnBurst_Instantaneous(
    Spawn Group=1
    Spawn Time=&RandomRangeFloat(Minimum=0.075, Maximum=0.1)
    Use Spawn Probability=false
    UseLoopCountLimit=false
   ) {module=SpawnBurst_Instantaneous002}
  ]
  particleSpawnScript [
   /SetVariables(Particles.VisibilityTag=&SpawnGroup()) {isSetParameters=true}
   &InitializeParticle(
    Color=<rgba 3 3 3 1>
    Lifetime=0.4
    Mass Mode=::NewEnumerator1
    Mesh Scale Mode=::NewEnumerator3
    Mesh Uniform Scale=&FloatFromCurve(
     FloatCurve=DataInterface() {propertyValues {Curve {keys [{tangentMode=RCTM_User, time=0, value=1}, {tangentMode=RCTM_User, time=1, value=2}]}}}
     CurveIndex=&ReturnNormalizedExecIndex()
     Scale Curve=6
    )
    Position=&SimulationPosition()
    Position Mode=::NewEnumerator1
    Sprite Size Mode=::NewEnumerator3
    Uniform Sprite Size=400
    UsePositionOffset=true
    Write Material Random=true
   ) {module=InitializeParticle001}
   &CameraOffset(Camera Offset Amount=50)
   &SpriteFacingAndAlignment(Sprite Alignment=<0 0 1>, WriteAlignment=true, WriteFacing=false)
   &InitialMeshOrientation(
    Mesh Orientation Mode=::NewEnumerator2
    Rotation=&SelectVectorFromArray(
     Direct Array Index=&SpawnGroup()
     Vector Selection Array=DataInterface() {propertyValues {FloatData [<-0.25 0 0>, <0.5 0 0>]}}
    )
    Rotation Coordinate Space=::NewEnumerator2
    Use Rotation Vector=true
   )
  ]
  particleUpdateScript [
   &ParticleState()
   &SolveForcesAndVelocity(Clamp Velocity=false, Limit Acceleration=false)
   &ScaleColor(
    Scale Alpha=&FloatFromCurve(FloatCurve=DataInterface() {propertyValues {Curve {keys [
         {arriveTangent=6.4019, interpMode=RCIM_Cubic, leaveTangent=6.4019, tangentMode=RCTM_User, time=0, value=0}
         {arriveTangent=0.002, interpMode=RCIM_Cubic, leaveTangent=0.002, time=0.2, value=0.7}
         {arriveTangent=-0.9006, interpMode=RCIM_Cubic, leaveTangent=-0.9006, tangentMode=RCTM_User, time=1.0011, value=0.002}
        ]}}})
    ScaleA=true
    ScaleRGB=false
   )
   &ScaleSpriteSize(Scale Factor=&Multiply_Vector2D_ByFloat(
     Float=&FloatFromCurve(
      FloatCurve=DataInterface() {propertyValues {Curve {keys [
          {arriveTangent=0.7585, interpMode=RCIM_Cubic, leaveTangent=0.7585, tangentMode=RCTM_User, time=0, value=0.4}
          {interpMode=RCIM_Cubic, tangentMode=RCTM_User, time=1, value=1}
         ]}}}
      Scale Curve=1.25
     )
     Vector2D=<1 1>
    ))
   &ScaleMeshSize(Scale Factor=&VectorFromFloat(Value=&FloatFromCurve(
      FloatCurve=DataInterface() {propertyValues {Curve {keys [
          {arriveTangent=15.9108, interpMode=RCIM_Cubic, time=0, value=0.0062}
          {arriveTangent=1.1862, interpMode=RCIM_Cubic, leaveTangent=1.1862, time=0.2014, value=0.709}
          {arriveTangent=-0.0943, interpMode=RCIM_Cubic, leaveTangent=-0.0943, time=0.5693, value=0.6815}
          {interpMode=RCIM_Cubic, leaveTangent=-0.1625, time=1, value=0.6336}
         ]}}}
      Scale Curve=&Multiply_Float(A=1.75, B=$User.Size_Scale)
     )))
   /SetVariables(
    Particles.DynamicParameter_01:Vector4f=&MakeVector4(
     X=&RandomRangeFloat(Minimum=0, Maximum=1)
     Y=2.5
     Z=&RandomRangeFloat(Minimum=2, Maximum=3)
     W=0
    )
    Particles.DynamicParameter_02:Vector4f=&MakeVector4(X=0, Y=&RandomRangeFloat(Minimum=0, Maximum=1), Z=0, W=0)
   ) {isSetParameters=true}
   /SetVariables(Particles.Corona_Size:Vector2f=&Multiply_Vector2D_ByFloat(Float=0.75, Vector2D=$Particles.SpriteSize)) {isSetParameters=true}
   /SetVariables(Particles.DarkGlowSize:Vector2f=&Multiply_Vector2D_ByFloat(Float=0.75, Vector2D=$Particles.SpriteSize)) {isSetParameters=true}
   /SetVariables(Particles.DargGlowColor:LinearColor=&ColorFromCurve(ColorCurve=DataInterface() {propertyValues {
       AlphaCurve {keys [<at 0.4134 0.7 RCIM_Cubic>,<at 1 0 RCIM_Cubic>]}
       BlueCurve {keys [<at 0 0.02 RCIM_Cubic>]}
       GreenCurve {keys [<at 0 0.0133 RCIM_Cubic>]}
       RedCurve {keys [<at 0 0.0158 RCIM_Cubic>]}
      }})) {isSetParameters=true}
   /SetVariables(Particles.CentralFlareSize:Vector2f=&Multiply_Vector2D_ByFloat(Float=1.5, Vector2D=$Particles.SpriteSize)) {isSetParameters=true}
   /SetVariables(Particles.CentralFlareColor:LinearColor=&MultiplyLinearColor(
     Linear Color=&ColorFromCurve(
      ColorCurve=DataInterface() {propertyValues {
        AlphaCurve {keys [<at 0 1>,<at 1 0>]}
        BlueCurve {keys [<at 0 0.2>,<at 0.333 0.2>,<at 0.667 1>,<at 1 0.2>]}
        GreenCurve {keys [<at 0 0.2>,<at 0.333 1>,<at 0.667 0.2>,<at 1 0.2>]}
        RedCurve {keys [<at 0 1>,<at 0.333 0.2>,<at 0.667 0.2>,<at 1 1>]}
       }}
      CurveIndex=&Multiply_Float(A=$Particles.NormalizedAge, B=2)
     )
     Scale Factor=&MakeFloat4FromVectorAndFloat("Vector (RGB)"=&VectorFromFloat(Value=10), "Float (Alpha)"=1)
    )) {isSetParameters=true}
  ]
  rendererClasses [&NiagaraSpriteRendererProperties, &NiagaraMeshRendererProperties]
  renderers [
   {rendererClass=&NiagaraSpriteRendererProperties}
   {rendererClass=&NiagaraSpriteRendererProperties, rendererIndex=1}
   {rendererClass=&NiagaraSpriteRendererProperties, rendererIndex=2}
   {rendererClass=&NiagaraSpriteRendererProperties, rendererIndex=3}
   {rendererClass=&NiagaraMeshRendererProperties, rendererIndex=4}
  ]
 }
 Rays {
  emitterUpdateScript [
   &EmitterState(Life Cycle Mode=::NewEnumerator1, Loop Behavior=::NewEnumerator1, Loop Duration=0.3, UseLoopDelay=true)
   &SpawnBurst_Instantaneous(Spawn Count=12, Use Spawn Probability=false, UseLoopCountLimit=false)
  ]
  particleSpawnScript [
   /SetVariables(Particles.SpawnRotationVector=&RandomRangeVector(Minimum=<-0.5 0 0>, Maximum=<0.5 0 0>)) {isSetParameters=true}
   /SetVariables(Particles.SpawnRotationVector=&RandomRangeVector(Minimum=<-1 -1 -1>, Maximum=<1 1 1>)) {isSetParameters=true}
   &InitializeParticle(
    Lifetime=&RandomRangeFloat(Minimum=0.25, Maximum=0.4)
    Mesh Scale=/Niagara/DynamicInputs/CascadeConversion/SplitVector(
     Double Component Vector Channels:ECascadeNiagaraTwoVectorChannels=::NewEnumerator1
     Double Component:NiagaraFloat=3.5
     Single Component:NiagaraFloat=1.4
    )
    Mesh Scale Mode=::NewEnumerator4
    UsePositionOffset=false
    Write Material Random=true
   )
   &SystemLocation(MaskSpawnGroup=false, Offset=$User.Offset)
   &CascadeConversion_InitialMeshRotation(Rotation:Vector3f=&RandomRangeVector(Minimum=<0 0 0>, Maximum=<1 1 1>))
  ]
  particleUpdateScript [
   &ParticleState()
   &ScaleMeshSize(Scale Factor=&SplitVector(
     Double Component Vector Channels:ECascadeNiagaraTwoVectorChannels=::NewEnumerator1
     Double Component:NiagaraFloat=&FloatFromCurve(FloatCurve=DataInterface() {propertyValues {Curve {keys [{arriveTangent=3.0317, interpMode=RCIM_Cubic, time=0, value=0.0432}, <at 1 1.6255 RCIM_Cubic>]}}})
     Single Component:NiagaraFloat=&FloatFromCurve(FloatCurve=DataInterface() {propertyValues {Curve {keys [{arriveTangent=0.1226, interpMode=RCIM_Cubic, time=0, value=0.0866}, <at 1 1.9984 RCIM_Cubic>]}}})
    )) {module=ScaleMeshSize001}
   &UpdateMeshOrientation(Rotation Vector=$Particles.SpawnRotationVector)
   &Color(Color=&MakeLinearColorFromVectorAndFloat(
     "Vector (RGB)"=&VectorFromFloat(Value=&FloatFromCurve(FloatCurve=DataInterface() {propertyValues {Curve {keys [<at 0 0.9964 RCIM_Cubic>, {arriveTangent=-0.96, interpMode=RCIM_Cubic, leaveTangent=-0.96, time=0.5, value=0.293}, <at 1 0.0364>]}}}))
     "Float (Alpha)"=&ClampFloat(Float=&FloatFromCurve(FloatCurve=DataInterface() {propertyValues {Curve {keys [
           <at -0.002 0.0024 RCIM_Cubic>
           {arriveTangent=-0.0024, interpMode=RCIM_Cubic, leaveTangent=-0.0024, time=0.1595, value=0.2}
           <at 1 0 RCIM_Cubic>
          ]}}}))
    ))
   &DynamicMaterialParameters(
    Index 0 Param 1=&FloatFromCurve(FloatCurve=DataInterface() {propertyValues {Curve {keys [{arriveTangent=-1.4398, interpMode=RCIM_Cubic, time=0, value=2.4783}, <at 1 1.5796 RCIM_Cubic>]}}})
    Index 0 Param 2=1
    Index 0 Param 3=&FloatFromCurve(FloatCurve=DataInterface() {propertyValues {Curve {keys [{arriveTangent=3.3568, interpMode=RCIM_Cubic, time=0, value=0.5548}, <at 1 1.5037>]}}})
    Index 0 Param 4=0.5
    Module.Index0Param0Write=true
    Module.Index0Param1Write=true
    Module.Index0Param2Write=true
    Module.Index0Param3Write=true
   )
   &CameraOffset(Camera Offset Amount=60)
  ]
  rendererClasses [&NiagaraMeshRendererProperties]
  renderers [{rendererClass=&NiagaraMeshRendererProperties, rendererIndex=0}]
 }
}
summary {
 systemName=NS_Proto_Crystal_Burst_01
 emitters [
  {emitterName=MainShape, rendererClasses [&NiagaraSpriteRendererProperties, &NiagaraMeshRendererProperties]}
  {emitterName=Rays, rendererClasses [&NiagaraMeshRendererProperties]}
  {emitterName=Crystal, rendererClasses [&NiagaraMeshRendererProperties]}
  {emitterName=Crystal_Large, rendererClasses [&NiagaraMeshRendererProperties]}
  {emitterName=Lines, rendererClasses [&NiagaraSpriteRendererProperties]}
 ]
 userVariables [
  User.ExplosionScale {type=NiagaraFloat, defaultValue=1, description=""}
  User.FX_Scale_Global {type=NiagaraFloat, defaultValue=0, description=""}
  User.FXColor {type=LinearColor, description="", defaultValue {struct=LinearColor, value=<rgba 0.9087 0.0025 1 1>}}
  User.FXScale {type=NiagaraFloat, defaultValue=500, description=""}
  User.FXTime {type=NiagaraFloat, defaultValue=0.5, description=""}
  User.Offset {type=Vector3f, description="", defaultValue {struct=Vector3f, value=<0 0 0>}}
  User.Size_Scale {type=NiagaraFloat, defaultValue=0.5, description=""}
  User.VelocityScale {type=NiagaraFloat, defaultValue=1, description=""}
 ]
}
