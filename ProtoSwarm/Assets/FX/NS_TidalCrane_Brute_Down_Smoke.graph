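# Niagara system for a smoke cloud effect, named for the TidalCrane Brute's "Down" event or state.
# It holds a single emitter, LW_Smoke_Cloud, whose prefix suggests a lightweight emitter; no renderers or emitter modules are listed.
# SystemState is the only module recorded.
# It exposes User.Ascended (bool) and User.IsAir (float) parameters, with no visible link to the emitter.
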NiagaraSystem /ProtoSwarm/Assets/FX/NS_TidalCrane_Brute_Down_Smoke
dependencies {usedDataInterfaces [], usedDynamicInputs [], usedModules [&SystemState]}
emitters {LW_Smoke_Cloud {rendererClasses [], renderers []}}
summary {
 systemName=NS_TidalCrane_Brute_Down_Smoke
 emitters [{emitterName=LW_Smoke_Cloud, rendererClasses []}]
 userVariables [
  User.Ascended {type=NiagaraBool}
  User.IsAir {type=NiagaraFloat, defaultValue=1, description=""}
 ]
}
