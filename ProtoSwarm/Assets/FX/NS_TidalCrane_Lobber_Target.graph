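# Niagara target marker, apparently telegraphing where a TidalCrane Lobber attack will land.
# Flash_Flare001 spawns one particle feeding two decal renderers, with an outer decal size shrinking over 1.5 seconds.
# Outer_to_Inner spawns mesh particles offset horizontally toward User.BomberLocation, sending them back toward the owner as they grow, shift color and fade, indicating the incoming direction.
# The Inner_to_Outer emitter is disabled.
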NiagaraSystem /ProtoSwarm/Assets/FX/NS_TidalCrane_Lobber_Target
dependencies=null
emitters {
 Flash_Flare001 {
  emitterUpdateScript [
   &EmitterState()
   &SpawnBurst_Instantaneous(Spawn Count=1, Use Spawn Probability=false)
  ]
  particleSpawnScript [
   &InitializeParticle(
    Color=<rgba 0.019 0.0287 0.0938 0.35>
    Lifetime=$Emitter.CurrentLoopDuration
    Position Offset=<0 0 -40>
    UsePositionOffset=true
    Write Material Random=true
   )
   /SetVariables(
    Particles.DecalFade=1
    Particles.DecalOrientation:Quat4f=&OrientMeshToVector(Look At Direction=<0 0 1>, Reference Axis Vector=<1 0 0>)
    Particles.DecalSize:Vector3f=<400 650 650>
   ) {isSetParameters=true}
  ]
  particleUpdateScript [
   &ParticleState()
   /SetVariables(
    Particles.DecalFade:NiagaraFloat=&Lerp_Float(A=0, B=1, Alpha=$System.DecalSize)
    Particles.Outer_Decak Size:Vector3f=&Lerp_Vector(
     A=<400 1100 1100>
     B=<400 580 580>
     Alpha=&FloatFromCurve(
      FloatCurve=DataInterface() {propertyValues {Curve {
         defaultValue=1
         keys [
          {arriveTangent=0.1993, interpMode=RCIM_Cubic, leaveTangent=0.1993, tangentMode=RCTM_User, time=0, value=0}
          {arriveTangent=0.7777, tangentMode=RCTM_User, time=1.5, value=1}
         ]
        }}}
      CurveIndex=$Emitter.Age
     )
    )
   ) {isSetParameters=true}
  ]
  rendererClasses [&NiagaraDecalRendererProperties]
  renderers [
   {rendererClass=&NiagaraDecalRendererProperties, rendererIndex=0}
   {rendererClass=&NiagaraDecalRendererProperties, rendererIndex=1}
  ]
 }
 Outer_to_Inner {
  emitterUpdateScript [
   &EmitterState(Life Cycle Mode=::NewEnumerator1, Loop Behavior=::NewEnumerator1, Loop Duration=2, UseLoopDelay=false)
   &SpawnRate(SpawnRate=3.5, Use Spawn Probability=false)
  ]
  particleSpawnScript [
   &InitializeParticle(
    Color Mode=::NewEnumerator0
    Mesh Scale=<1.5 0.75 1>
    Mesh Scale Mode=::NewEnumerator4
    Position Offset=&AddVectorToPosition(
     Position=&Multiply_VectorByFloat(
      Vector=&NormalizeVector(
       Vector To Normalize=&Multiply_Vector(
        A=&SubtractVectorFromPosition(Position=$User.BomberLocation, Vector=Linked() {linkedVariable {name=Engine.Owner.Position, type=NiagaraPosition}})
        B=<1 1 0>
       )
       Safe Normalize=true
       Safe Fallback Threshold=0
       Safe Fallback Vector=<0 0 0.001>
      )
      Float=300
     )
     Vector=<0 0 4>
    )
    Position Offset Coordinate Space=::World
    UsePositionOffset=true
    Write Material Random=true
   )
   &CameraOffset(Camera Offset Amount=30)
   &AddVelocity(
    Velocity=&NormalizeVector(
     Vector To Normalize=&Multiply_Vector(
      A=&SubtractVectorFromPosition(Position=$User.BomberLocation, Vector=Linked() {linkedVariable {name=Engine.Owner.Position, type=NiagaraPosition}})
      B=<1 1 0>
     )
     Safe Normalize=true
     Safe Fallback Threshold=0
     Safe Fallback Vector=<0 0 0.001>
    )
    Velocity Speed Scale=-200
   )
   &ApplyInitialForces()
   &InitialMeshOrientation(Mesh Orientation Mode=::NewEnumerator2, Rotation Coordinate Space=::NewEnumerator1, Use Rotation Vector=false)
  ]
  particleUpdateScript [
   &ParticleState()
   &ScaleMeshSize(Scale Factor=&Multiply_VectorByFloat(
     Vector=<1 1 1>
     Float=&FloatFromCurve(
      FloatCurve=DataInterface() {propertyValues {Curve {keys [
          {arriveTangent=2.1948, interpMode=RCIM_Cubic, leaveTangent=2.1948, tangentMode=RCTM_User, time=0, value=0.1}
          {tangentMode=RCTM_User, time=1, value=1}
         ]}}}
      Scale Curve=2
     )
    ))
   &Drag(Drag=0.7, Use Linear Drag=true, Use Rotational Drag=true)
   &SpriteFacingAndAlignment(Facing Coordinate Space=::World, Sprite Facing=<0 0 1>, WriteAlignment=false, WriteFacing=true)
   &SolveForcesAndVelocity(Clamp Velocity=false, Limit Acceleration=false)
   &ScaleColor(
    Linear Color Curve=DataInterface() {propertyValues {
      AlphaCurve {keys [<at 0 0>,<at 0.1698 1>,<at 1 0>]}
      BlueCurve {keys [<at 0 1>,<at 0.3111 3.7796>,<at 0.7328 1>]}
      GreenCurve {keys [<at 0 0.0835>,<at 0.3111 0>,<at 0.7328 0>]}
      RedCurve {keys [<at 0 0>,<at 0.3111 4>,<at 0.7328 0.1422>]}
     }}
    Scale Mode=::NewEnumerator2
   )
   &DynamicMaterialParameters(
    Index 0 Param 1=&Add_Float(A=&RandomRangeFloat(Minimum=-0.2, Maximum=0.25), B=0)
    Index 0 Param 2=&RandomRangeFloat(Minimum=4, Maximum=3)
    Index 0 Param 3=&RandomRangeFloat(Minimum=2, Maximum=1)
    Index 1 Param 4=15
    Index 2 Param 1=5
    Index 3 Param 1=&RandomRangeFloat(Minimum=1, Maximum=2)
    Index 3 Param 2=&RandomRangeFloat(Minimum=0, Maximum=1)
    Index 3 Param 3=55
    Index 3 Param 4=55
    Module.Param0WriteEnabled=true
    Module.Param0WriteEnabled1=false
    Module.Param0WriteEnabled2=false
    Module.Param0WriteEnabled3=true
    Module.Param1WriteEnabled=true
    Module.Param1WriteEnabled1=false
    Module.Param1WriteEnabled2=false
    Module.Param1WriteEnabled3=true
    Module.Param2WriteEnabled=true
    Module.Param2WriteEnabled1=false
    Module.Param2WriteEnabled2=false
    Module.Param2WriteEnabled3=false
    Module.Param3WriteEnabled=true
    Module.Param3WriteEnabled1=false
    Module.Param3WriteEnabled2=false
    Module.Param3WriteEnabled3=false
    Write Parameter Index 1=true
    Write Parameter Index 2=true
    Write Parameter Index 3=true
   )
   &ScaleColor(
    Scale Alpha=&FloatFromCurve(FloatCurve=DataInterface() {propertyValues {Curve {keys [<at 1 1>,<at 1.8 0>]}}}, CurveIndex=$Emitter.Age, Scale Curve=1.1)
    ScaleA=true
    ScaleRGB=false
   ) {module=ScaleColor001}
  ]
  rendererClasses [&NiagaraMeshRendererProperties]
  renderers [{rendererClass=&NiagaraMeshRendererProperties, rendererIndex=0}]
 }
}
summary {
 systemName=NS_TidalCrane_Lobber_Target
 emitters [
  {emitterName=Flash_Flare001, rendererClasses [&NiagaraDecalRendererProperties]}
  {emitterName=Outer_to_Inner, rendererClasses [&NiagaraMeshRendererProperties]}
  {bEnabled=false, emitterName=Inner_to_Outer, rendererClasses [&NiagaraMeshRendererProperties]}
 ]
 userVariables [User.BomberLocation {type=NiagaraPosition, description="", defaultValue {}}]
}
