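# Niagara ribbon trail for the TidalCrane projectile, judging by its name, built around a short-lived GooTrail emitter.
# Custom FN ribbon modules spawn and position particles from the User.SkeletalMesh interface.
# Curves taper width and fade alpha; a color lerp is wired into ScaleColor, but RGB scaling is off.
# GooTrail001 is a disabled, longer-lived variant using InitializeRibbon, plain SpawnRate and two ribbon renderers.
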NiagaraSystem /ProtoSwarm/Assets/FX/NS_TidalCrane_Prj_Trail
dependencies {usedDataInterfaces [&NiagaraDataInterfaceCurve], usedDynamicInputs [], usedModules [&SystemState]}
emitters {
 GooTrail {
  emitterUpdateScript [
   &EmitterState()
   /Game/Effects/Niagara/Modules/Ribbons/FN_SpawnRibbon_v3(Skeleton Mesh:NiagaraDataInterfaceSkeletalMesh=$User.SkeletalMesh, Spawn Mode:ENiagaraSkeletalSpawning=::NewEnumerator1, SpawnRate=40)
   /SetVariables(Emitter.DepthFadeDistance=250) {isSetParameters=true}
  ]
  particleSpawnScript [
   /SetVariables(Particles.ColorStart:LinearColor=<rgba 1 0.0459 0 1>) {isSetParameters=true, module=SetVariables_#1}
   /Game/Effects/Niagara/Modules/Ribbons/FN_InitializeRibbon_v5(
    Enable Skeletal Position=true
    Facing Axis:ENiagaraOrientationAxis=::YAxis
    Lifetime=0.4
    Offset Amount=25
    Offset Axis:ENiagaraOrientationAxis=::XAxis
    Ribbon Width=325
    SkeletalMesh:NiagaraDataInterfaceSkeletalMesh=$User.SkeletalMesh
   )
  ]
  particleUpdateScript [
   /Game/Effects/Niagara/Modules/Ribbons/FN_UpdateRibbon_v5()
   /SetVariables(Particles.RibbonFacing=<1 0 0>) {isSetParameters=true, module=SetVariables_#2}
   &ScaleColor(
    Scale Alpha=&FloatFromCurve(FloatCurve=DataInterface() {propertyValues {Curve {keys [
         {interpMode=RCIM_Cubic, tangentMode=RCTM_Break, time=0, value=1}
         {
          arriveTangent=-0.7676
          arriveTangentWeight=0.0667
          interpMode=RCIM_Cubic
          leaveTangent=-0.7676
          leaveTangentWeight=0.4021
          tangentMode=RCTM_User
          time=0.1
          value=0.5
         }
         {interpMode=RCIM_Cubic, tangentMode=RCTM_Break, time=1, value=0.3}
        ]}}})
    Scale RGB=&MakeVectorFromLinearColorRGB(Color=&LerpLinearColors(
      StartColor=$Particles.ColorStart
      EndColor=$Particles.ColorEnd
      LerpFactor=&FloatFromCurve(
       FloatCurve=DataInterface() {propertyValues {Curve {keys [
           {interpMode=RCIM_Cubic, leaveTangent=-0.0081, tangentMode=RCTM_Break, time=0, value=0}
           {arriveTangent=2.8653, interpMode=RCIM_Cubic, leaveTangent=2.8653, tangentMode=RCTM_User, time=0.1, value=0.2}
           {interpMode=RCIM_Cubic, tangentMode=RCTM_Break, time=0.7, value=1}
          ]}}}
       CurveIndex=$Particles.Age
      )
     ))
    ScaleA=true
    ScaleRGB=false
   )
   &ScaleRibbonWidth(Ribbon Width Scale=&FloatFromCurve(FloatCurve=DataInterface() {propertyValues {Curve {keys [
         {interpMode=RCIM_Cubic, tangentMode=RCTM_Break, time=0.2, value=1}
         {arriveTangent=-1.6921, interpMode=RCIM_Cubic, tangentMode=RCTM_Break, time=1, value=0.4}
        ]}}}))
  ]
  rendererClasses [&NiagaraRibbonRendererProperties]
  renderers [{rendererClass=&NiagaraRibbonRendererProperties}]
 }
 GooTrail001 {
  bEnabled=false
  emitterUpdateScript [
   &EmitterState()
   &FN_SpawnRibbon_v3(Spawn Mode:ENiagaraSkeletalSpawning=::NewEnumerator1, SpawnRate=40) {enabled=false}
   &SpawnRate(SpawnRate=50, Use Spawn Probability=false)
  ]
  particleSpawnScript [
   /SetVariables(Particles.ColorStart:LinearColor=<rgba 1 0.0459 0 1>) {isSetParameters=true, module=SetVariables_#1}
   &FN_InitializeRibbon_v5(
    Enable Skeletal Position=true
    Facing Axis:ENiagaraOrientationAxis=::YAxis
    Lifetime=4
    Offset Amount=25
    Offset Axis:ENiagaraOrientationAxis=::XAxis
    Ribbon Width=325
    SkeletalMesh:NiagaraDataInterfaceSkeletalMesh=$User.SkeletalMesh
   ) {enabled=false}
   &InitializeRibbon(
    Lifetime=4
    Ribbon Facing Mode:ENiagara_UnsetDirectSet=::NewEnumerator1
    Ribbon Facing Vector:Vector3f=<0 1 0>
    Ribbon Width=325
    UsePositionOffset=false
    Write Material Random=true
   )
  ]
  particleUpdateScript [
   &FN_UpdateRibbon_v5() {enabled=false}
   &SolveForcesAndVelocity(Clamp Velocity=false, Limit Acceleration=false)
   &FN_UpdateRibbon(Detach Locked Particle=true)
   &ParticleState()
   /SetVariables(Particles.RibbonFacing=<1 0 0>) {isSetParameters=true, module=SetVariables_#2}
   &ScaleColor(
    Scale Alpha=&FloatFromCurve(FloatCurve=DataInterface() {propertyValues {Curve {keys [
         {interpMode=RCIM_Cubic, tangentMode=RCTM_Break, time=0, value=1}
         {
          arriveTangent=-0.7676
          arriveTangentWeight=0.0667
          interpMode=RCIM_Cubic
          leaveTangent=-0.7676
          leaveTangentWeight=0.4021
          tangentMode=RCTM_User
          time=0.1
          value=0.5
         }
         {interpMode=RCIM_Cubic, tangentMode=RCTM_Break, time=1, value=0.3}
        ]}}})
    Scale RGB=&MakeVectorFromLinearColorRGB(Color=&LerpLinearColors(
      StartColor=$Particles.ColorStart
      EndColor=$Particles.ColorEnd
      LerpFactor=&FloatFromCurve(
       FloatCurve=DataInterface() {propertyValues {Curve {keys [
           {interpMode=RCIM_Cubic, leaveTangent=-0.0081, tangentMode=RCTM_Break, time=0, value=0}
           {arriveTangent=2.8653, interpMode=RCIM_Cubic, leaveTangent=2.8653, tangentMode=RCTM_User, time=0.1, value=0.2}
           {interpMode=RCIM_Cubic, tangentMode=RCTM_Break, time=0.7, value=1}
          ]}}}
       CurveIndex=$Particles.Age
      )
     ))
    ScaleA=true
    ScaleRGB=false
   )
   &FN_RibbonMaterial() {enabled=false}
  ]
  rendererClasses [&NiagaraRibbonRendererProperties]
  renderers [
   {rendererClass=&NiagaraRibbonRendererProperties}
   {rendererClass=&NiagaraRibbonRendererProperties, rendererIndex=1}
  ]
 }
}
summary {
 systemName=NS_TidalCrane_Prj_Trail
 emitters [
  {emitterName=GooTrail, rendererClasses [&NiagaraRibbonRendererProperties]}
  {bEnabled=false, emitterName=GooTrail001, rendererClasses [&NiagaraRibbonRendererProperties]}
 ]
 userVariables [
  User.bIsFrontEndPreview {type=NiagaraBool}
  User.PlayerSpeed {type=NiagaraFloat, defaultValue=0, description=""}
  User.RibbonIDOffset {type=NiagaraInt32}
  User.SkeletalMesh {
   type=NiagaraDataInterfaceSkeletalMesh
   defaultValue {
    struct=NiagaraExt_VariableValue_DataInterface
    value {dataInterface=&self:NiagaraDataInterfaceSkeletalMesh_0}
   }
  }
 ]
}
