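# M_BugGoo is a Default Lit master material for bug goo surfaces.
# It outputs material attributes from a single call to the MF_BugGoo material function, which holds all the shading logic.
# It is flagged for use with Nanite, and automatic usage setting in the editor is turned off.
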Material /ProtoSwarm/Assets/Materials/M_BugGoo
graph {
 properties {bAutomaticallySetUsageInEditor=false, bUseMaterialAttributes=true, bUsedWithNanite=true, usedShadingModels=Default Lit}
 nodes [MaterialFunctionCall() {materialFunction=&/MF_BugGoo}]
}
