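# M_BugGooColorMaster is a Default Lit master material for animated, glowing bug goo, with parameters for color, roughness and animation.
# Two panning caustic samples, one rotating, multiply into a pattern shared by the vertex motion and glow.
# SimpleGrassWind times that pattern is sphere-masked and blended by layered sine waves for vertex wobble.
# Pulsing Fresnel masks blend edge and goo colors into emissive, scaled by time of day and dimmest at night.
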Material /ProtoSwarm/Assets/Materials/M_BugGooColorMaster
graph {
 properties {bAutomaticallySetUsageInEditor=false, usedShadingModels=Default Lit}
 nodes [
  n1 = ScalarParameter() {defaultValue=2, parameterName=VertAnimIntensity}
  n2 = ScalarParameter() {defaultValue=0.15, parameterName=VertAnimSpeed}
  n3 = Panner() {speedX=0.035}
  n4 = Panner(Coordinate=Rotator() {speed=0.05}) {speedY=0.036}
  n5 = Multiply(
   A=ScalarParameter() {defaultValue=3.5, parameterName=AnimCausticsMulti}
   B=Multiply(
    A=TextureSample(UVs=@n3) {samplerType=SAMPLERTYPE_LinearGrayscale, texture=/Game/Effects/Fort_Effects/Textures/Tile/T_Caustic_Wind_Diagonal}.RGB
    B=TextureSample(UVs=@n4) {samplerType=SAMPLERTYPE_LinearGrayscale, texture=&T_Caustic_Wind_Diagonal}.RGB
   )
  )
  n6 = Multiply(A=Multiply(
    A=MaterialFunctionCall!(WindIntensity=@n1, WindWeight=@n2, WindSpeed=@n2, AdditionalWPO=@n1) {materialFunction=/Engine/Functions/Engine_MaterialFunctions01/WorldPositionOffset/SimpleGrassWind}.Result
    B=@n5
   )) {constB=1.25}
  n7 = MaterialFunctionCall!(Hardness=ScalarParameter() {defaultValue=0.1, parameterName=MaskHardness}, Center, Radius=Constant2Vector() {g=0.4, r=0.4}) {materialFunction=/Game/Packages/Fortress_SharedMaterials/Base_Material_Functions/FX_SphereMask2D}
  n8 = Sine(None=Time()) {period=6}
  n9 = Sine(None=Time()) {period=11.2}
  n10 = Abs(None=Cosine(None=Time()) {period=21})
  Abs(None=LinearInterpolate(A=Multiply(A=@n6, B=@n7.Result), B=Multiply(A=@n7.Result) {constB=25}, Alpha=LinearInterpolate(A=@n8, B=@n9, Alpha=@n10) {constB=0.01}) {constB=0.01})
  n11 = Fresnel() {exponent=2}
  n12 = MaterialFunctionCall(Hardness=ScalarParameter() {defaultValue=0.05, parameterName=FrenselHardness}, Radius=Constant2Vector() {g=0.35, r=0.35}) {materialFunction=&FX_SphereMask2D}
  n13 = Fresnel() {exponent=1.75}
  n14 = MaterialFunctionCall(Hardness=ScalarParameter() {defaultValue=0.05, parameterName=FrenselHardness}, Radius=Constant2Vector() {g=0.5, r=0.5}) {materialFunction=&FX_SphereMask2D}
  n15 = Time()
  n16 = LinearInterpolate(
   A=VectorParameter() {parameterName=EdgeColorColor, defaultValue=<rgba 0.0003 0.099 0.0128 1>}.RGB
   B=VectorParameter() {parameterName=GooColor, defaultValue=<rgba 0 0.5469 0.0238 1>}.RGB
   Alpha=LinearInterpolate(A=Multiply(A=@n11, B=@n12.Result), B=Multiply(A=@n13, B=@n14.Result), Alpha=Sine(None=@n15) {period=9})
  )
  n17 = Constant() {r=0.5}
  Multiply(
   A=Multiply(A=ScalarParameter() {defaultValue=1.2, parameterName=EmmisiveMulti}, B=Multiply(A=@n16, B=@n5))
   B=MaterialFunctionCall(Value_Morning=@n17, Value_Day=Constant() {r=1}, Value_Evening=@n17, Value_Night=Constant() {r=0.15}) {materialFunction=&Scalar_Time_of_Day}.Result
  )
  ScalarParameter() {defaultValue=0.1, parameterName=Roughness}
 ]
}
