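# GrassTiling is a material function that blends two world-tiled grass layers to vary ground cover.
# Each layer takes the Z-projection of WorldAlignedTextures_Complex at its own tile size, with overridable diffuse, spec mask and normal textures.
# Grass 01's spec mask alpha, resampled over world XY at larger scale, drives the blend.
# A static switch can substitute vertex normals, then a Fortnite spec swizzle feeds the attribute and per-channel outputs.
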MaterialFunction /Game/Packages/Fortnite_UniqueMaterials_1/DS_Fortnight/GrassTiling
graph {nodes [
  n1 = TextureObject() {texture=/Game/Packages/Fortress_Floors_Generic/Textures/Grass0043_2_L}
  n2 = FunctionInput(Preview=TextureObject() {samplerType=SAMPLERTYPE_Masks, texture=/Game/Packages/Fortnite_UniqueMaterials_1/Textures/T_Grass_01_SpecMask_Grass12LerpInAlpha}) {bUsePreviewValueAsDefault=true, inputName="Grass 01 SpecMask (Alpha Contains Grass 1 2 Lerp Mask)", inputType=FunctionInput_Texture2D}
  n3 = TextureObject() {samplerType=SAMPLERTYPE_Normal, texture=/Game/Packages/Fortnite_UniqueMaterials_1/Textures/Water1_NRM}
  n4 = StaticBool() {value=true}
  n5 = TextureObject() {texture=/Game/Packages/DS_Fortnight/T_FORT_Grass02}
  n6 = TextureObject() {samplerType=SAMPLERTYPE_Masks, texture=&T_Fortnite_Default_S}
  n7 = TextureObject() {samplerType=SAMPLERTYPE_Normal, texture=&Water1_NRM}
  n8 = ComponentMask(None=WorldPosition() {worldPositionShaderOffset=WPT_ExcludeAllShaderOffsets}.XYZ) {g=true, r=true}
  n9 = Constant() {r=2048}
  n10 = MaterialFunctionCall!(
   Base Material=MaterialFunctionCall!(
    Diffuse Texture Object=FunctionInput(Preview=@n1) {bUsePreviewValueAsDefault=true, inputName=Grass 01, inputType=FunctionInput_Texture2D}
    Spec Texture Object=@n2
    TextureSize=Constant() {r=1024}
    --------------------------
    Use seperate settings for Z projection
    Use seperate texture for Z projection
    Texture Size Z projection
    Z Plane Diffuse Texture Object
    Z Plane Spec Texture Object
    "--------------------------      "
    ProjectionTransitionContrast
    Use Texture Alpha Heightmap to Refine the Transistion
    HeightLerpTransitionContrast
    PickTheDiffuseHeightChannel
    TheDiffuseHeightChannel
    "--------------------------       "
    NormalTextureObject=FunctionInput(Preview=@n3) {bUsePreviewValueAsDefault=true, inputName=GrassNormal_01, inputType=FunctionInput_Texture2D, sortPriority=10}
    World Space Normals
    "--------------------------  "
    Export Float 4 Diffuse=@n4
    Export Float 4 Spec
    "--------------------------     "
    WorldPosition
    World Space Normal
   ) {materialFunction=/Engine/Functions/Engine_MaterialFunctions02/Texturing/WorldAlignedTextures_Complex}.Z Mat Attributes - See Tooltip
   Top Material=MaterialFunctionCall(
    Diffuse Texture Object=FunctionInput(Preview=@n5) {bUsePreviewValueAsDefault=true, inputName=Grass 02, inputType=FunctionInput_Texture2D, sortPriority=2}
    Spec Texture Object=FunctionInput(Preview=@n6) {bUsePreviewValueAsDefault=true, inputName=Grass 02 SpecMask, inputType=FunctionInput_Texture2D, sortPriority=2}
    TextureSize=Constant() {r=512}
    NormalTextureObject=FunctionInput(Preview=@n7) {bUsePreviewValueAsDefault=true, inputName=GrassNormal_02, inputType=FunctionInput_Texture2D, sortPriority=11}
    Export Float 4 Diffuse=@n4
   ) {materialFunction=&WorldAlignedTextures_Complex}.Z Mat Attributes - See Tooltip
   Alpha=TextureSample(UVs=Divide(A=@n8, B=@n9), Tex=@n2) {texture=/Game/Packages/DS_Fortnight/T_TerrainMasks}.A
  ) {materialFunction=/Engine/Functions/MaterialLayerFunctions/MatLayerBlend_Standard}
  n11 = VertexNormalWS()
  n12 = StaticBool() {value=true}
  n13 = MaterialFunctionCall!(In=StaticSwitch(
    True=@n10.Blended Material
    False=MaterialFunctionCall!(Material=@n10.Blended Material, Normal=@n11) {materialFunction=/Engine/Functions/MaterialLayerFunctions/MatLayerBlend_ReplaceNormals}.Result
    Value=FunctionInput(Preview=@n12) {bUsePreviewValueAsDefault=true, inputName=Use Grass Normals, inputType=FunctionInput_StaticBool}
   )) {materialFunction=/Game/Packages/Fortress_SharedMaterials/Base_Material_Functions/SwizzleWorldAlignedSpecOutputForFortnite}
  n14 = BreakMaterialAttributes!(Attr=@n13.Result)
  FunctionOutput(None=@n14.BaseColor) {outputName=Diffuse}
  FunctionOutput(None=@n13.Result) {outputName=MaterialAttributes - fortnite format}
  FunctionOutput(None=@n14.Normal) {outputName=Normals}
  FunctionOutput(None=@n14.Roughness) {outputName=Roughness}
  FunctionOutput(None=@n14.Specular) {outputName=Spec}
  FunctionOutput(None=@n14.Metallic) {outputName=Metallic}
 ]}
