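# Layered terrain master material that blends grass or base, mud, concrete, dirt, top and two cliff layers, plus an optional masked paint layer.
# Vertex color channels drive HeightLerp blends, using each layer's opacity as height, with darkening at mud and dirt transitions.
# Most layers use world-aligned projection, with UV alternatives for cliffs, concrete and top layer.
# Static switches toggle layers and features, while quality switches simplify mud shading and normals on Low.
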Material /ProtoSwarm/Assets/Materials/Terrain_Mat
graph {
 properties {
  bTangentSpaceNormal=false
  bUsedWithMeshParticles=true
  physMaterial=/Game/FortressPhysicalMaterials/Dirt
  pixelDepthOffsetMode=PDOM_Legacy
  refractionMethod=RM_IndexOfRefraction
  usedShadingModels=Default Lit
 }
 nodes [
  n1 = TextureObjectParameter() {
   group=Concrete
   parameterName=Concrete
   samplerSource=SSM_Wrap_WorldGroupSettings
   texture=/Game/Packages/DS_Fortnight/SM/Textures/T_Street_Road_D
  }
  n2 = ComponentMask(None=MaterialFunctionCall(TextureObject=@n1, TextureSize=Constant() {desc=concrete, r=768}, Export Float 4=StaticBool() {value=true}) {materialFunction=&WorldAlignedTexture}.Z Texture) {a=true}
  n3 = Constant() {r=2}
  n4 = ComponentMask(None=VertexColor().R) {desc=Transition phase for concrete, r=true}
  Clamp(None=Multiply(
    A=MaterialFunctionCall!(Lerp A, Lerp B, Height Texture=@n2, Falloff Contrast=@n3, Transition Phase=@n4) {materialFunction=/Game/Packages/Fortnight_MaterialFunctions/Temp/HeightLerp}.Lerp Alpha
    B=Constant() {r=5}
   ))
  Constant() {r=0.5}
  n5 = TextureObjectParameter() {
   group=Cliffs
   parameterName=CliffTexture
   samplerSource=SSM_Wrap_WorldGroupSettings
   texture=/Game/Packages/DS_Fortnite_Terrain_NoLOD/Textures/T_Cave_Exterior_Wall_D
  }
  n6 = TextureObjectParameter() {
   group=Cliffs
   parameterName=Cliff Spec Texture
   samplerSource=SSM_Wrap_WorldGroupSettings
   samplerType=SAMPLERTYPE_Masks
   texture=/Game/Packages/Fortnite_UniqueMaterials_1/Textures/T_Fortnite_Default_S
  }
  n7 = StaticBool() {value=true}
  n8 = TextureObjectParameter() {
   group=Cliffs
   parameterName=CliffNormal
   samplerSource=SSM_Wrap_WorldGroupSettings
   samplerType=SAMPLERTYPE_Normal
   texture=&FlatNormalBlue
  }
  n9 = WorldPosition()
  n10 = ComponentMask(None=@n9.XYZ) {g=true, r=true}
  n11 = ComponentMask(None=@n9.XYZ) {b=true}
  n12 = Multiply(A=@n11, B=ScalarParameter() {defaultValue=0.25, group=Cliffs, parameterName=XY coordinate offset over Z}) {constB=0.25}
  n13 = Multiply(
   A=MaterialFunctionCall(f2=@n10) {materialFunction=&AddComponents}.f2
   B=ScalarParameter() {defaultValue=0.25, group=Cliffs, parameterName=Z coordinate offset over XY}
  ) {constB=0.25}
  n14 = MaterialFunctionCall!(
   Diffuse Texture Object=@n5
   Spec Texture Object=@n6
   TextureSize=Abs(None=ScalarParameter() {defaultValue=-768, group=Cliffs, parameterName=CliffWorldCoordScale})
   --------------------------
   Use seperate settings for Z projection
   Use seperate texture for Z projection
   Texture Size Z projection
   Z Plane Diffuse Texture Object
   Z Plane Spec Texture Object
   "--------------------------      "
   ProjectionTransitionContrast=ScalarParameter() {group=Cliffs, parameterName=Cliff Projection Contrast}
   Use Texture Alpha Heightmap to Refine the Transistion=@n7
   HeightLerpTransitionContrast=ScalarParameter() {defaultValue=5, group=Cliffs, parameterName=Cliff Height Lerp Contrast}
   PickTheDiffuseHeightChannel
   TheDiffuseHeightChannel
   "--------------------------       "
   NormalTextureObject=@n8
   World Space Normals=@n7
   "--------------------------  "
   Export Float 4 Diffuse=@n7
   Export Float 4 Spec
   "--------------------------     "
   WorldPosition=AppendVector(A=Add(A=@n10, B=@n12), B=Add(A=@n13, B=@n11))
   World Space Normal
  ) {materialFunction=/Engine/Functions/Engine_MaterialFunctions02/Texturing/WorldAlignedTextures_Complex}
  n15 = MaterialFunctionCall!(In=StaticSwitchParameter(True=@n14.XYZ Mat Attributes - See Tooltip, False=@n14.XYZ Mat Attributes - See Tooltip) {defaultValue=true, group=Cliffs, parameterName="Use XYZ Cliff Projection. False=XY"}) {materialFunction=/Game/Packages/Fortress_SharedMaterials/Base_Material_Functions/SwizzleWorldAlignedSpecOutputForFortnite}
  n16 = TextureSampleParameter2D() {
   group=Cliffs
   parameterName=CliffTexture
   samplerSource=SSM_Wrap_WorldGroupSettings
   texture=&T_Cave_Exterior_Wall_D
  }
  n17 = TextureSampleParameter2D() {
   group=Cliffs
   parameterName=Cliff Spec Texture
   samplerSource=SSM_Wrap_WorldGroupSettings
   samplerType=SAMPLERTYPE_Masks
   texture=&T_Fortnite_Default_S
  }
  n18 = TextureSampleParameter2D() {
   group=Cliffs
   parameterName=CliffNormal
   samplerSource=SSM_Wrap_WorldGroupSettings
   samplerType=SAMPLERTYPE_Normal
   texture=/Game/Packages/DS_Fortnite_Terrain_NoLOD/Textures/T_Cave_Exterior_Wall_N
  }
  n19 = MakeMaterialAttributes!(
   BaseColor=@n16.RGB
   Metallic=@n17.G
   Specular=@n17.G
   Roughness=@n17.G
   Anisotropy
   EmissiveColor
   Opacity=@n16.RGB
   OpacityMask
   Normal=Transform(None=@n18.RGB)
   Tangent
   WorldPositionOffset
   SubsurfaceColor
   ClearCoat
   ClearCoatRoughness
   AmbientOcclusion
   Refraction
   CustomizedUVs_0
   CustomizedUVs_1
   CustomizedUVs_2
   CustomizedUVs_3
   CustomizedUVs_4
   CustomizedUVs_5
   CustomizedUVs_6
   CustomizedUVs_7
   PixelDepthOffset
   ShadingModel
   Displacement
  )
  n20 = MaterialFunctionCall!(Mat attributes=MaterialFunctionCall!(
    Mat attributes=StaticSwitchParameter(True=@n15.Result, False=@n19) {defaultValue=true, group=Cliffs, parameterName=Use World Aligned Cliff Textures}
    Use Spec=StaticBoolParameter() {group=Cliffs, parameterName=Use Cliff Spec}
    Use Normals=StaticBoolParameter() {defaultValue=true, group=Cliffs, parameterName=Use Cliff Normals}
   ) {materialFunction=/Game/Packages/Fortnite_UniqueMaterials_1/MaterialFunctions/Fortnite_Terrain_invertOpacityNormalBool}.Result) {materialFunction=/Engine/Functions/Engine_MaterialFunctions02/MatLayerBlend_BreakOpacity}
  n21 = TextureObjectParameter() {group=Mud, parameterName=Mud_Diffuse, samplerSource=SSM_Wrap_WorldGroupSettings, texture=/Game/Packages/DS_Fortnight/Mud}
  n22 = TextureObjectParameter() {
   group=Mud
   parameterName=Mud_SpecMask
   samplerSource=SSM_Wrap_WorldGroupSettings
   samplerType=SAMPLERTYPE_Masks
   texture=&T_Fortnite_Default_S
  }
  n23 = Abs(None=ScalarParameter() {defaultValue=512, group=Mud, parameterName=MudTextureTilingFrequency})
  n24 = TextureObjectParameter() {
   group=Mud
   parameterName=Mud_Normal
   samplerSource=SSM_Wrap_WorldGroupSettings
   samplerType=SAMPLERTYPE_Normal
   texture=/Game/Packages/DS_Fortnight/mudnormal
  }
  n25 = StaticBool() {value=true}
  n26 = MaterialFunctionCall(
   Mat attributes=MaterialFunctionCall(In=MaterialFunctionCall(Diffuse Texture Object=@n21, Spec Texture Object=@n22, TextureSize=@n23, NormalTextureObject=@n24, Export Float 4 Diffuse=@n25) {materialFunction=&WorldAlignedTextures_Complex}.Z Mat Attributes - See Tooltip) {materialFunction=&SwizzleWorldAlignedSpecOutputForFortnite}.Result
   Use Spec=StaticBoolParameter() {group=Mud, parameterName=Use Mud Spec}
   Use Normals=StaticBoolParameter() {group=Mud, parameterName=Use Mud Normals}
  ) {materialFunction=&Fortnite_Terrain_invertOpacityNormalBool}
  n27 = VectorParameter() {group=Mud, parameterName=CustomMud_Sub_Mult_DarkeningLevel_Contrast, defaultValue=<rgba 0.6 0.8 0.1 0>}
  n28 = VertexColor()
  n29 = OneMinus(None=Multiply(A=@n28.R) {constB=2}) {desc=1- -1}
  n30 = Multiply(B=@n28.R) {constA=-2, constB=-2}
  n31 = StaticSwitchParameter(
   True=StaticSwitchParameter(True=Clamp(None=@n29), False=Constant()) {defaultValue=true, group=Concrete, parameterName=UseConcrete}
   False=Clamp(None=Add(A=@n28.A, B=@n30) {desc=0 to -2})
  ) {group=Cliffs, parameterName="UseCliffs (Disables Mud)"}
  n32 = MaterialFunctionCall(Mat attributes=@n26.Result) {materialFunction=&MatLayerBlend_BreakOpacity}
  n33 = TextureObjectParameter() {
   group=Grass
   parameterName=Grass01_SpecMask
   samplerSource=SSM_Wrap_WorldGroupSettings
   samplerType=SAMPLERTYPE_Masks
   texture=/Game/Packages/Fortnite_UniqueMaterials_1/Textures/T_Grass_01_SpecMask_Grass12LerpInAlpha
  }
  n34 = TextureObjectParameter() {group=Grass, parameterName=Grass01_D, samplerSource=SSM_Wrap_WorldGroupSettings, texture=/Game/Packages/DS_Fortnight/T_FORT_Grass02}
  n35 = TextureObjectParameter() {
   group=Grass
   parameterName=Grass02_SpecMask
   samplerSource=SSM_Wrap_WorldGroupSettings
   samplerType=SAMPLERTYPE_Masks
   texture=/Game/Packages/Fortress_Floors_Generic/Textures/Grass0043_2_M
  }
  n36 = TextureObjectParameter() {
   group=Grass
   parameterName=Grass02_D
   samplerSource=SSM_Wrap_WorldGroupSettings
   texture=/Game/Packages/Fortress_Floors_Generic/Textures/Grass0043_2_L
  }
  n37 = TextureObjectParameter() {group=Grass, parameterName=GrassNormal_01, samplerSource=SSM_Wrap_WorldGroupSettings, samplerType=SAMPLERTYPE_Normal, texture=null}
  n38 = TextureObjectParameter() {group=Grass, parameterName=GrassNormal_02, samplerSource=SSM_Wrap_WorldGroupSettings, samplerType=SAMPLERTYPE_Normal, texture=null}
  n39 = MaterialFunctionCall(
   Mat attributes=MaterialFunctionCall!(
    "Grass 01 SpecMask (Alpha Contains Grass 1 2 Lerp Mask)"=@n33
    Use Grass Normals
    Grass 01=@n34
    Grass 02 SpecMask=@n35
    Grass 02=@n36
    GrassNormal_01=@n37
    GrassNormal_02=@n38
   ) {materialFunction=/Game/Packages/Fortnite_UniqueMaterials_1/DS_Fortnight/GrassTiling}.MaterialAttributes - fortnite format
   Use Spec=StaticBoolParameter() {group=Grass, parameterName=Use Grass Spec}
   Use Normals=StaticBoolParameter() {group=Grass, parameterName=Use Grass Normals}
  ) {materialFunction=&Fortnite_Terrain_invertOpacityNormalBool}
  n40 = TextureObjectParameter() {
   group=BaseTilingTexture
   parameterName=Base Tiling Texture
   samplerSource=SSM_Wrap_WorldGroupSettings
   texture=/Game/Packages/DS_Fortnite_Terrain_NoLOD/Textures/T_FORT_Underground_02
  }
  n41 = TextureObjectParameter() {
   group=BaseTilingTexture
   parameterName=Base Tiling Texture_SpecMask
   samplerSource=SSM_Wrap_WorldGroupSettings
   samplerType=SAMPLERTYPE_Masks
   texture=&T_Fortnite_Default_S
  }
  n42 = TextureObjectParameter() {
   group=BaseTilingTexture
   parameterName=Base Tiling Normal Texture
   samplerSource=SSM_Wrap_WorldGroupSettings
   samplerType=SAMPLERTYPE_Normal
   texture=/Game/Packages/Fortnite_UniqueMaterials_1/T_Normal
  }
  n43 = MaterialFunctionCall(
   Diffuse Texture Object=@n40
   Spec Texture Object=@n41
   TextureSize=ScalarParameter() {defaultValue=-512, group=BaseTilingTexture, parameterName=Base Tiling Texture Tile Size}
   NormalTextureObject=@n42
   Export Float 4 Diffuse=StaticBool() {value=true}
  ) {materialFunction=&WorldAlignedTextures_Complex}
  n44 = MaterialFunctionCall(
   Mat attributes=MaterialFunctionCall(In=StaticSwitchParameter(True=@n43.XYZ Mat Attributes - See Tooltip, False=@n43.XYZ Mat Attributes - See Tooltip) {group=BaseTilingTexture, parameterName="Make Base Texture project XYZ(true) or Z(False)"}) {materialFunction=&SwizzleWorldAlignedSpecOutputForFortnite}.Result
   Use Spec=StaticBoolParameter() {group=BaseTilingTexture, parameterName=Use Base Texture Spec}
   Use Normals=StaticBoolParameter() {group=BaseTilingTexture, parameterName=Use Base Texture Normals}
  ) {materialFunction=&Fortnite_Terrain_invertOpacityNormalBool}
  n45 = Constant()
  n46 = Constant() {r=1}
  n47 = MaterialFunctionCall(Mat attributes=StaticSwitchParameter(
    True=StaticSwitchParameter(True=@n39.Result, False=@n44.Result) {defaultValue=true, group=Grass, parameterName="Use Grass(true) or a BaseTilingTexture(False)"}
    False=MakeMaterialAttributes(BaseColor=@n45, Specular=@n45, Opacity=@n46)
   ) {defaultValue=true, group=Grass, parameterName=Use a base Layer}) {desc=can remove later, materialFunction=&Fortnite_Terrain_invertOpacityNormalBool}
  n48 = Multiply(
   A=Constant() {r=0.5}
   B=MaterialFunctionCall(Mat attributes=@n47.Result) {desc=Grass Or Base Opacity, materialFunction=&MatLayerBlend_BreakOpacity}.Result
  )
  n49 = MaterialFunctionCall(Transition Phase=@n31, Height Texture=Clamp(None=Add(A=@n32.Result, B=@n48)), Contrast=Constant() {r=3}) {desc=mud + grass blended, materialFunction=/Engine/Functions/Engine_MaterialFunctions02/Texturing/HeightLerp}
  n50 = Constant()
  n51 = StaticSwitchParameter(True=@n49.Lerp Alpha No Contrast, False=@n50) {defaultValue=true, group=Concrete, parameterName=UseConcrete}
  n52 = StaticSwitchParameter(
   True=@n49.Lerp Alpha No Contrast
   False=StaticSwitchParameter(True=@n49.Lerp Alpha No Contrast, False=@n50) {group=Mud, parameterName=UseMud}
  ) {defaultValue=true, group=Concrete, parameterName=UseConcrete}
  n53 = OneMinus(None=ComponentMask(None=StaticSwitchParameter(True=@n51, False=@n52) {group=Cliffs, parameterName="UseCliffs (Disables Mud)"}) {desc=height lerp lerp alpha no contrast from mud, r=true})
  n54 = Subtract(A=@n53, B=@n27.R) {constB=0.4}
  n55 = SquareRoot(None=@n31) {desc=0}
  n56 = Clamp(None=Add(A=Multiply(A=@n54, B=@n27.G) {constA=1, constB=0.5}, B=OneMinus(None=@n55)))
  n57 = StaticSwitchParameter(
   True=MaterialFunctionCall(
    Transition Phase=@n56
    Height Texture=MaterialFunctionCall(Mat attributes=@n26.Result) {materialFunction=&MatLayerBlend_BreakOpacity}.Result
    Contrast=@n27.A
   ) {materialFunction=&HeightLerp}.Alpha
   False=@n56
  ) {group=Mud, parameterName=Hlerp_MudDarking}
  n58 = MaterialFunctionCall!(Material=@n26.Result, Tint=@n27.B, TintMask=@n57) {materialFunction=/Engine/Functions/MaterialLayerFunctions/MatLayerBlend_Tint}
  n59 = OneMinus(None=Multiply(A=@n57, B=ScalarParameter() {group=Mud, parameterName=Mud Wetness Level}))
  n60 = OneMinus(None=Clamp(None=Add(A=@n53) {constB=0.1}))
  n61 = Multiply(A=@n60, B=@n60)
  n62 = Add(A=OneMinus(None=Multiply(A=@n61) {constB=5})) {constB=0.25}
  n63 = StaticSwitchParameter(
   True=MaterialFunctionCall(
    Material=MaterialFunctionCall!(Base Material=@n58.Blended Material, RoughnessMultiplier=@n59) {desc=New, materialFunction=/Engine/Functions/MaterialLayerFunctions/MatLayerBlend_ModulateRoughness}.Blended Material
    Tint=Clamp(None=@n62)
   ) {desc=grass shadow, materialFunction=&MatLayerBlend_Tint}.Blended Material
   False=@n58.Blended Material
  ) {defaultValue=true, group=Grass, parameterName=Grass Shadow}
  n64 = MaterialFunctionCall!(Base Material=QualitySwitch(Default=@n63, Low=@n26.Result, High=@n63), Top Material=@n47.Result, Alpha=@n49.Alpha) {materialFunction=/Engine/Functions/MaterialLayerFunctions/MatLayerBlend_Standard}
  n65 = StaticSwitchParameter(True=@n64.Blended Material, False=StaticSwitchParameter(True=@n64.Blended Material, False=@n47.Result) {group=Mud, parameterName=UseMud}) {defaultValue=true, group=Concrete, parameterName=UseConcrete}
  n66 = MaterialFunctionCall!(Concrete diffuse, Concrete Spec, Concrete Normal) {materialFunction=/Game/Packages/DS_Fortnight/SM/Materials/concreteMF}
  n67 = StaticSwitchParameter(
   True=MaterialFunctionCall(Base Material=@n65, Top Material=@n66.Material Attribute, Alpha=@n66.Material Attribute) {materialFunction=&MatLayerBlend_Standard}.Blended Material
   False=@n65
  ) {defaultValue=true, group=Concrete, parameterName=UseConcrete}
  n68 = TextureObjectParameter() {group=Dirt, parameterName=Dirt, samplerSource=SSM_Wrap_WorldGroupSettings, texture=/Game/Packages/DS_Fortnight/whiteRock}
  n69 = TextureObjectParameter() {
   group=Dirt
   parameterName=Dirt_SpecMask
   samplerSource=SSM_Wrap_WorldGroupSettings
   samplerType=SAMPLERTYPE_Masks
   texture=&T_Fortnite_Default_S
  }
  n70 = ScalarParameter() {defaultValue=1024, group=Dirt, parameterName=DirtTextureTilingFrequency}
  n71 = TextureObjectParameter() {
   group=Dirt
   parameterName=DirtTextureNormal
   samplerSource=SSM_Wrap_WorldGroupSettings
   samplerType=SAMPLERTYPE_Normal
   texture=&T_Normal
  }
  n72 = StaticBool() {value=true}
  n73 = MaterialFunctionCall(
   Mat attributes=MaterialFunctionCall(In=MaterialFunctionCall(Diffuse Texture Object=@n68, Spec Texture Object=@n69, TextureSize=@n70, NormalTextureObject=@n71, Export Float 4 Diffuse=@n72) {materialFunction=&WorldAlignedTextures_Complex}.Z Mat Attributes - See Tooltip) {materialFunction=&SwizzleWorldAlignedSpecOutputForFortnite}.Result
   Use Spec=StaticBoolParameter() {group=Dirt, parameterName=Use Dirt Spec}
   Use Normals=StaticBoolParameter() {group=Dirt, parameterName=Use Dirt Normals}
  ) {materialFunction=&Fortnite_Terrain_invertOpacityNormalBool}
  n74 = MaterialFunctionCall(Material=@n73.Result, Tint=VectorParameter() {group=Dirt, parameterName=DirtColor, defaultValue=<rgba 0.221 0.173 0.1111 1>}.RGB) {materialFunction=&MatLayerBlend_Tint}
  n75 = MaterialFunctionCall(Mat attributes=@n47.Result) {materialFunction=&MatLayerBlend_BreakOpacity}
  n76 = StaticSwitchParameter(True=@n49.Alpha, False=@n50) {group=Mud, parameterName=UseMud}
  n77 = OneMinus(None=StaticSwitchParameter(
    True=OneMinus(None=@n75.Result)
    False=StaticSwitchParameter(True=@n49.Alpha, False=@n76) {defaultValue=true, group=Concrete, parameterName=UseConcrete}
   ) {group=Cliffs, parameterName="UseCliffs (Disables Mud)"}) {desc=inverted mud transition alpha}
  n78 = MaterialFunctionCall(Mat attributes=@n73.Result) {materialFunction=&MatLayerBlend_BreakOpacity}
  n79 = MaterialFunctionCall(Transition Phase=VertexColor().B, Height Texture=Clamp(None=ComponentMask(None=@n78.Result) {r=true}), Contrast=Constant() {r=1}) {materialFunction=&HeightLerp}
  n80 = MaterialFunctionCall(Material=@n74.Blended Material, Tint=Clamp(None=Add(A=@n77, B=@n79.Alpha))) {materialFunction=&MatLayerBlend_Tint}
  n81 = Constant() {r=0.5}
  n82 = StaticSwitchParameter(
   True=MaterialFunctionCall(
    Base Material=@n67
    Top Material=StaticSwitchParameter(True=@n80.Blended Material, False=@n74.Blended Material) {defaultValue=true, group=Dirt, parameterName=Darken edges around dirt transitions}
    Alpha=MaterialFunctionCall(In=@n79.Lerp Alpha No Contrast, Contrast=@n81) {materialFunction=&CheapContrast}.Result
   ) {materialFunction=&MatLayerBlend_Standard}.Blended Material
   False=@n67
  ) {defaultValue=true, group=Dirt, parameterName=UseDirt}
  n83 = MaterialFunctionCall(Mat attributes=@n82) {materialFunction=&MatLayerBlend_BreakOpacity}
  n84 = ScalarParameter() {defaultValue=0.5, group=Cliffs, parameterName=Grass/Cliff Transition. Rock or Grass H map lerp}
  n85 = ScalarParameter() {defaultValue=5, group=Cliffs, parameterName=CliffLerpContrast}
  n86 = MaterialFunctionCall(
   Transition Phase=Power(
    Base=VertexColor() {desc=mask cliff to grass transition}.A
    Exp=ScalarParameter() {defaultValue=2, group=Cliffs, parameterName=Cliff to Other Layers Transition Power}
   )
   Height Texture=OneMinus(None=LinearInterpolate(A=@n20.Result, B=@n83.Result, Alpha=@n84))
   Contrast=@n85
  ) {materialFunction=&HeightLerp}
  LinearInterpolate(
   B=MaterialFunctionCall(In=@n86.Lerp Alpha No Contrast, Contrast=Multiply(A=@n85) {constB=0.5}) {materialFunction=&CheapContrast}.Result
   Alpha=ScalarParameter() {defaultValue=1, group=Cliffs, parameterName=Shadow on Cliff Amount}
  ) {constA=1}
  n87 = TextureObjectParameter() {
   group=Concrete
   parameterName=Concrete
   samplerSource=SSM_Wrap_WorldGroupSettings
   texture=&T_Street_Road_D
  }
  n88 = TextureObjectParameter() {
   group=Concrete
   parameterName=Concrete_SpecMask
   samplerSource=SSM_Wrap_WorldGroupSettings
   samplerType=SAMPLERTYPE_Masks
   texture=/Game/Packages/DS_Fortnight/SM/Textures/T_Street_Road_M
  }
  n89 = TextureObjectParameter() {
   group=Concrete
   parameterName=ConcreteTextureNormal
   samplerSource=SSM_Wrap_WorldGroupSettings
   samplerType=SAMPLERTYPE_Normal
   texture=/Game/Packages/DS_Fortnight/SM/Textures/T_Street_Road_N
  }
  n90 = MaterialFunctionCall(
   Diffuse Texture Object=@n87
   Spec Texture Object=@n88
   TextureSize=Constant() {r=1024}
   NormalTextureObject=@n89
   Export Float 4 Diffuse=StaticBool() {value=true}
  ) {materialFunction=&WorldAlignedTextures_Complex}
  n91 = TextureCoordinate()
  n92 = VectorParameter() {group=Concrete, parameterName=ConcreteTextureCoordinateControls_RG_Scale_BA_Offset, defaultValue=<rgba 1 1 0 0>}
  n93 = ComponentMask(None=@n92.RGB) {g=true, r=true}
  n94 = StaticSwitchParameter(True=Add(A=Multiply(A=@n91, B=@n93), B=AppendVector(A=@n92.B, B=@n92.B)), False=@n91) {group=Concrete, parameterName=Modify Concrete UVS}
  n95 = TextureSampleParameter2D(UVs=@n94) {
   group=Concrete
   parameterName=Concrete
   samplerSource=SSM_Wrap_WorldGroupSettings
   texture=&T_Street_Road_D
  }
  n96 = TextureSampleParameter2D() {
   group=Concrete
   parameterName=Concrete_SpecMask
   samplerSource=SSM_Wrap_WorldGroupSettings
   texture=&T_Street_Road_M
  }
  n97 = TextureSampleParameter2D(UVs=@n94) {
   group=Concrete
   parameterName=ConcreteTextureNormal
   samplerSource=SSM_Wrap_WorldGroupSettings
   samplerType=SAMPLERTYPE_Normal
   texture=&T_Street_Road_N
  }
  n98 = StaticSwitch(True=@n97.RGB, False=VertexNormalWS(), Value=StaticBoolParameter() {defaultValue=true, group=Concrete, parameterName=Use Concrete Normals})
  MaterialFunctionCall(
   Mat attributes=StaticSwitchParameter(
    True=MaterialFunctionCall(In=@n90.Z Mat Attributes - See Tooltip) {materialFunction=&SwizzleWorldAlignedSpecOutputForFortnite}.Result
    False=MakeMaterialAttributes(BaseColor=@n95.RGB, Metallic=@n96.G, Specular=@n96.G, Roughness=@n96.G, Opacity=@n95.RGB, Normal=@n98)
   ) {defaultValue=true, group=Concrete, parameterName=UseWorldPositionCoordinatesForConcrete}
   Use Spec=StaticBoolParameter() {defaultValue=true, group=Concrete, parameterName=Use Concrete Spec}
   Use Normals=StaticBoolParameter() {defaultValue=true, group=Concrete, parameterName=Use Concrete Normals}
  ) {materialFunction=&Fortnite_Terrain_invertOpacityNormalBool}
  n99 = TextureObjectParameter() {
   group=TopLayer
   parameterName=Top Lerp Diffuse
   samplerSource=SSM_Wrap_WorldGroupSettings
   texture=/Game/Packages/DS_Fortnight/LeavesDead0019_D
  }
  n100 = TextureObjectParameter() {
   group=TopLayer
   parameterName=Top Lerp Spec Mask
   samplerSource=SSM_Wrap_WorldGroupSettings
   samplerType=SAMPLERTYPE_Masks
   texture=&T_Fortnite_Default_S
  }
  n101 = TextureObjectParameter() {
   group=TopLayer
   parameterName=Top Lerp Normal
   samplerSource=SSM_Wrap_WorldGroupSettings
   samplerType=SAMPLERTYPE_Normal
   texture=&T_Normal
  }
  n102 = MaterialFunctionCall(
   Diffuse Texture Object=@n99
   Spec Texture Object=@n100
   TextureSize=ScalarParameter() {defaultValue=-512, group=TopLayer, parameterName=Top Layer Tiling Frequency}
   NormalTextureObject=@n101
   Export Float 4 Diffuse=StaticBool() {value=true}
  ) {materialFunction=&WorldAlignedTextures_Complex}
  n103 = TextureSampleParameter2D() {
   group=TopLayer
   parameterName=Top Lerp Diffuse
   samplerSource=SSM_Wrap_WorldGroupSettings
   texture=&LeavesDead0019_D
  }
  n104 = TextureSampleParameter2D() {
   group=TopLayer
   parameterName=Top Lerp Spec Mask
   samplerSource=SSM_Wrap_WorldGroupSettings
   samplerType=SAMPLERTYPE_Masks
   texture=&T_Fortnite_Default_S
  }
  n105 = TextureSampleParameter2D() {
   group=TopLayer
   parameterName=Top Lerp Normal
   samplerSource=SSM_Wrap_WorldGroupSettings
   samplerType=SAMPLERTYPE_Normal
   texture=&T_Normal
  }
  n106 = Transform(None=@n105.RGB)
  n107 = MaterialFunctionCall(
   Mat attributes=StaticSwitchParameter(
    True=MaterialFunctionCall(In=@n102.Z Mat Attributes - See Tooltip) {materialFunction=&SwizzleWorldAlignedSpecOutputForFortnite}.Result
    False=MakeMaterialAttributes(BaseColor=@n103.RGB, Metallic=@n104.G, Specular=@n104.G, Roughness=@n104.G, Opacity=@n103.RGB, Normal=@n106)
   ) {defaultValue=true, group=TopLayer, parameterName=Use World Aligned Top Layer Textures}
   Use Spec=StaticBoolParameter() {group=TopLayer, parameterName=Use Top Lerp Spec}
   Use Normals=StaticBoolParameter() {group=TopLayer, parameterName=Use Top Lerp Normals}
  ) {materialFunction=&Fortnite_Terrain_invertOpacityNormalBool}
  n108 = StaticSwitchParameter(
   True=MaterialFunctionCall(Material=@n107.Result, Tint=VectorParameter() {group=TopLayer, parameterName=TopLayerColorMultiplier, defaultValue=<rgba 0 0 0 1>}.RGB) {materialFunction=&MatLayerBlend_Tint}.Blended Material
   False=@n107.Result
  ) {group=TopLayer, parameterName=Modulate Top Layer Diffuse}
  n109 = VertexColor()
  n110 = TextureObjectParameter() {
   group=TopLayer
   parameterName=Top Tiling Mask Texture
   samplerSource=SSM_Wrap_WorldGroupSettings
   samplerType=SAMPLERTYPE_LinearColor
   texture=null
  }
  n111 = MaterialFunctionCall(TextureObject=@n110, TextureSize=ScalarParameter() {defaultValue=-1024, group=TopLayer, parameterName=Top Layer Mask Tiling Frequency}) {materialFunction=&WorldAlignedTexture}
  n112 = OneMinus(None=Multiply(A=ScalarParameter() {defaultValue=0.75, group=TopLayer, parameterName="Top Tiling Mask Texture Raise White Level(0-1)"}, B=@n111.Z Texture) {constB=0.75})
  n113 = StaticSwitchParameter(
   True=Multiply(A=@n109.G, B=StaticSwitchParameter(True=@n112, False=@n111.Z Texture) {group=TopLayer, parameterName=Modify Top Tiling Mask Brightness})
   False=@n109.G
  ) {group=TopLayer, parameterName=Modulate Top Level Transition Values With a texture}
  n114 = MaterialFunctionCall(Mat attributes=@n107.Result) {materialFunction=&MatLayerBlend_BreakOpacity}
  n115 = ScalarParameter() {defaultValue=256, group=TopLayer, parameterName=Top Layer Blend Contrast}
  n116 = StaticSwitchParameter(
   True=MaterialFunctionCall(
    Base Material=@n82
    Top Material=@n108
    Alpha=MaterialFunctionCall(A=@n82, B=@n108, Transition Phase=@n113, Height Texture=@n114.Result, Contrast=@n115) {materialFunction=&HeightLerp}.Alpha
   ) {materialFunction=&MatLayerBlend_Standard}.Blended Material
   False=@n82
  ) {desc=diffuse, group=TopLayer, parameterName="UseAdditionalTopLayerBlend(ControlledByGreenVertColor)"}
  n117 = TextureObjectParameter() {
   group=Cliffs_2
   parameterName=CliffTexture_2
   samplerSource=SSM_Wrap_WorldGroupSettings
   texture=&T_Cave_Exterior_Wall_D
  }
  n118 = TextureObjectParameter() {
   group=Cliffs_2
   parameterName=Cliff Spec Texture_2
   samplerSource=SSM_Wrap_WorldGroupSettings
   samplerType=SAMPLERTYPE_Masks
   texture=&T_Fortnite_Default_S
  }
  n119 = StaticBool() {value=true}
  n120 = TextureObjectParameter() {
   group=Cliffs_2
   parameterName=CliffNormal_2
   samplerSource=SSM_Wrap_WorldGroupSettings
   samplerType=SAMPLERTYPE_Normal
   texture=&FlatNormalBlue
  }
  n121 = ComponentMask(None=WorldPosition().XYZ) {g=true, r=true}
  n122 = ComponentMask(None=WorldPosition().XYZ) {b=true}
  n123 = Multiply(A=@n122, B=ScalarParameter() {defaultValue=0.25, group=Cliffs_2, parameterName=XY coordinate offset over Z_2}) {constB=0.25}
  n124 = Multiply(
   A=MaterialFunctionCall(f2=@n121) {materialFunction=&AddComponents}.f2
   B=ScalarParameter() {defaultValue=0.25, group=Cliffs_2, parameterName=Z coordinate offset over XY_2}
  ) {constB=0.25}
  n125 = MaterialFunctionCall(
   Diffuse Texture Object=@n117
   Spec Texture Object=@n118
   TextureSize=ScalarParameter() {defaultValue=-768, group=Cliffs_2, parameterName=CliffWorldCoordScale_2}
   ProjectionTransitionContrast=ScalarParameter() {group=Cliffs_2, parameterName=Cliff Projection Contrast_2}
   Use Texture Alpha Heightmap to Refine the Transistion=@n119
   HeightLerpTransitionContrast=ScalarParameter() {defaultValue=5, group=Cliffs_2, parameterName=Cliff Height Lerp Contrast_2}
   NormalTextureObject=@n120
   World Space Normals=@n119
   Export Float 4 Diffuse=@n119
   WorldPosition=AppendVector(A=Add(A=@n121, B=@n123), B=Add(A=@n124, B=@n122))
  ) {materialFunction=&WorldAlignedTextures_Complex}
  n126 = StaticSwitchParameter(True=@n125.XYZ Mat Attributes - See Tooltip, False=@n125.XYZ Mat Attributes - See Tooltip) {defaultValue=true, group=Cliffs_2, parameterName="Use XYZ Cliff_2 Projection. False=XY"}
  n127 = TextureSampleParameter2D() {
   group=Cliffs_2
   parameterName=CliffTexture_2
   samplerSource=SSM_Wrap_WorldGroupSettings
   texture=&T_Cave_Exterior_Wall_D
  }
  n128 = TextureSampleParameter2D() {
   group=Cliffs_2
   parameterName=Cliff Spec Texture_2
   samplerSource=SSM_Wrap_WorldGroupSettings
   samplerType=SAMPLERTYPE_Masks
   texture=&T_Fortnite_Default_S
  }
  n129 = TextureSampleParameter2D() {
   group=Cliffs_2
   parameterName=CliffNormal_2
   samplerSource=SSM_Wrap_WorldGroupSettings
   samplerType=SAMPLERTYPE_Normal
   texture=&T_Cave_Exterior_Wall_N
  }
  n130 = Transform(None=@n129.RGB)
  n131 = MaterialFunctionCall(
   Mat attributes=StaticSwitchParameter(
    True=MaterialFunctionCall(In=@n126) {materialFunction=&SwizzleWorldAlignedSpecOutputForFortnite}.Result
    False=MakeMaterialAttributes(BaseColor=@n127.RGB, Metallic=@n128.G, Specular=@n128.G, Roughness=@n128.G, Opacity=@n127.RGB, Normal=@n130)
   ) {defaultValue=true, group=Cliffs_2, parameterName=Use World Aligned Cliff 2 Textures}
   Use Spec=StaticBoolParameter() {group=Cliffs_2, parameterName=Use Cliff 2 Spec}
   Use Normals=StaticBoolParameter() {defaultValue=true, group=Cliffs_2, parameterName=Use Cliff 2 Normals}
  ) {materialFunction=&Fortnite_Terrain_invertOpacityNormalBool}
  n132 = VertexColor()
  n133 = MaterialFunctionCall(Mat attributes=@n131.Result) {materialFunction=&MatLayerBlend_BreakOpacity}
  n134 = LinearInterpolate(
   A=@n133.Result
   B=@n83.Result
   Alpha=ScalarParameter() {defaultValue=0.25, group=Cliffs_2, parameterName=Grass/Cliff Transition. Rock or Grass H map lerp_2}
  )
  n135 = MaterialFunctionCall(
   Base Material=@n116
   Top Material=@n131.Result
   Alpha=MaterialFunctionCall(
    Transition Phase=ComponentMask(None=@n132.A) {r=true}
    Height Texture=OneMinus(None=@n134)
    Contrast=ScalarParameter() {defaultValue=2, group=Cliffs_2, parameterName=CliffLerpContrast_2}
   ) {materialFunction=&HeightLerp}.Alpha
  ) {materialFunction=&MatLayerBlend_Standard}
  n136 = MaterialFunctionCall(Base Material=@n116, Top Material=MaterialFunctionCall(), Alpha=@n86.Alpha) {materialFunction=&MatLayerBlend_Standard}
  n137 = VertexColor()
  n138 = MaterialFunctionCall(
   Transition Phase=@n137.G
   Height Texture=LinearInterpolate(A=OneMinus(None=@n133.Result), B=@n20.Result, Alpha=@n137.G)
   Contrast=ScalarParameter() {defaultValue=30, group=Cliffs, parameterName=2 Cliff Transition contrast}
  ) {materialFunction=&HeightLerp}
  n139 = StaticSwitchParameter(
   True=StaticSwitchParameter(
    True=MaterialFunctionCall(Base Material=@n135.Blended Material, Top Material=@n136.Blended Material, Alpha=@n138.Alpha) {materialFunction=&MatLayerBlend_Standard}.Blended Material
    False=@n136.Blended Material
   ) {desc=Diffuse, group=Cliffs, parameterName="Lerp 2 Cliffs (Uses Green Channel To Lerp)"}
   False=@n116
  ) {desc=Diffuse, group=Cliffs, parameterName="UseCliffs (Disables Mud)"}
  n140 = MakeMaterialAttributes(
   BaseColor=VectorParameter() {group=PaintLayer, parameterName=PaintLayerColor, defaultValue=<rgba 1 1 1 1>}.RGB
   Metallic=ScalarParameter() {group=PaintLayer, parameterName=Paint Metallic}
   Roughness=ScalarParameter() {defaultValue=0.25, group=PaintLayer, parameterName=Paint Roughness}
   Normal=VertexNormalWS()
  )
  n141 = TextureCoordinate()
  n142 = VectorParameter() {group=PaintLayer, parameterName=PaintLayerUV_RG_Offset_BA_Scale, defaultValue=<rgba 0 0 1 1>}
  n143 = ComponentMask(None=@n142.RGB) {g=true, r=true}
  n144 = TextureSampleParameter2D(UVs=Multiply(A=Add(A=@n141, B=@n143), B=AppendVector(A=@n142.B, B=@n142.B))) {
   group=PaintLayer
   parameterName=PaintLayerMasks
   samplerType=SAMPLERTYPE_Masks
   texture=/Game/Packages/DS_Fortnight/SM/Textures/T_SoccerTennisBasketball
  }
  n145 = StaticComponentMaskParameter(None=AppendVector(A=@n144.RGB, B=@n144.RGB)) {defaultR=true, group=PaintLayer, parameterName=PaintLayerMaskChannel}
  n146 = WorldPosition()
  n147 = TextureSampleParameter2D(UVs=Divide(
    A=ComponentMask(None=@n146.XYZ) {g=true, r=true}
    B=ScalarParameter() {defaultValue=2048, group=PaintLayer, parameterName=PaintLayerMask_Mask Tiling}
   ) {constB=1024}) {
   group=PaintLayer
   parameterName=PaintLayerMasks_Mask
   samplerType=SAMPLERTYPE_Masks
   texture=/Game/Packages/Fortnite_UniqueMaterials_1/Textures/T_FRT_Conc_Masks_B_new
  }
  n148 = AppendVector(A=@n147.RGB, B=@n147.RGB)
  n149 = StaticSwitchParameter(
   True=Multiply(A=@n145, B=StaticComponentMaskParameter(None=@n148) {defaultR=true, group=PaintLayer, parameterName=PaintLayerMask_Mask_Channel})
   False=@n145
  ) {desc=Diffuse, group=PaintLayer, parameterName=Use Second PaintLayer Mask}
  n150 = StaticSwitchParameter(True=Multiply(A=@n149, B=ScalarParameter() {defaultValue=1, group=PaintLayer, parameterName=PaintLayer Alpha Modulation}), False=@n149) {desc=Diffuse, group=PaintLayer, parameterName=Modulate Paint Layer Mask}
  n151 = BreakMaterialAttributes!(Attr=StaticSwitchParameter(
    True=MaterialFunctionCall(Base Material=@n139, Top Material=@n140, Alpha=@n150) {materialFunction=&MatLayerBlend_Standard}.Blended Material
    False=@n139
   ) {desc=Diffuse, group=PaintLayer, parameterName=Use PaintLayer})
  MaterialFunctionCall!(DiffuseFresnelAmount=Constant() {r=0.25}, Emissive=@n151.EmissiveColor, Base Color=@n151.EmissiveColor) {materialFunction=/Game/Packages/Fortress_SharedMaterials/Base_Material_Functions/FresnelFunction}
  Multiply()
  StaticSwitchParameter(True=@n151.Specular, False=Constant()) {group=Specular, parameterName=Use Spec}
  n152 = VertexNormalWS()
  StaticSwitchParameter(True=QualitySwitch(Default=@n151.Normal, Low=@n152, High=@n151.Normal), False=@n152) {defaultValue=true, group=Normals, parameterName=Use Normals}
 ]
}
